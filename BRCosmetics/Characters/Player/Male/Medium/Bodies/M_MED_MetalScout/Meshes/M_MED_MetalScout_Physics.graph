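# Physics asset for the MetalScout medium male player body, set up around its dynamic skirt bones.
# Capsules cover the pelvis, both thighs, the right calf only, and two-segment skirt chains around front, sides and back, with no front-side left chain.
# Skirt roots constrain to the pelvis and tips to roots, with very tight twist limits.
# Front-left and front-middle tip joints are fully locked, front-right partly, while the right knee swings and twists freely.
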PhysicsAsset /BRCosmetics/Characters/Player/Male/Medium/Bodies/M_MED_MetalScout/Meshes/M_MED_MetalScout_Physics
bodies {
 calf_r [{
   shapeType=Capsule
   length=26.8341
   radius=9.7632
   shapeName=calf_r_capsule
   center=<26.8909 0.634 -0.1885>
   rotation=<rot 0 -84.1229 -89.1195>
  }]
 dyn_skirt_bk_l_a [{
   shapeType=Capsule
   length=5.3294
   radius=3.6706
   shapeName=dyn_skirt_bk_l_a_capsule
   center=<7.6448 -0.0178 -0.6127>
   rotation=<rot 7.6493 -89.148 -85.7587>
  }]
 dyn_skirt_bk_l_b [{shapeType=Capsule, length=5, radius=3, shapeName=dyn_skirt_bk_l_b_capsule, center=<6.4742 0 -0.3012>, rotation=<rot 87.3364 -0.0001 -0.0001>}]
 dyn_skirt_bk_r_a [{shapeType=Capsule, length=5.3294, radius=3.6706, center=<9.5851 -0.119 0.5472>, rotation=<rot 0.5332 -88.8856 -86.891>}]
 dyn_skirt_bk_r_b [{shapeType=Capsule, length=5, radius=3, shapeName=dyn_skirt_bk_r_b_capsule, center=<6.4742 0 -0.3012>, rotation=<rot 87.3364 0 0.0001>}]
 dyn_skirt_bk_side_l_a [{shapeType=Capsule, length=6, radius=3, center=<10.6351 0.1135 0.3589>, rotation=<rot -88.067 1.6918 178.9204>}]
 dyn_skirt_bk_side_l_b [{shapeType=Capsule, length=5, radius=3, shapeName=dyn_skirt_bk_side_l_b_capsule, center=<6.4742 0 -0.3012>, rotation=<rot 87.3365 0 -0.0003>}]
 dyn_skirt_bk_side_r_a [{shapeType=Capsule, length=6, radius=3, center=<9.6712 -0.0273 -0.5849>, rotation=<rot 0.3292 -91.1845 -86.5445>}]
 dyn_skirt_bk_side_r_b [{
   shapeType=Capsule
   length=5
   radius=3
   shapeName=dyn_skirt_bk_side_r_b_capsule
   center=<9.2102 0 -2.0465>
   rotation=<rot 87.3365 0.0001 0.0004>
  }]
 dyn_skirt_fr_l_a [{shapeType=Capsule, length=7, radius=5.5, center=<-21.0072 7.4737 0>, rotation=<rot 0 -9.2412 0>}]
 dyn_skirt_fr_l_b [{shapeType=Capsule, length=7, radius=5.5, center=<-21.7961 9.8589 0>, rotation=<rot 0 -9.2412 0>}]
 dyn_skirt_fr_mid_a [{shapeType=Capsule, length=6, radius=3, shapeName=dyn_skirt_fr_mid_a_capsule, center=<9.6588 0 -1.5281>, rotation=<rot 79.2825 0 0>}]
 dyn_skirt_fr_mid_b [{shapeType=Capsule, length=6, radius=3, shapeName=dyn_skirt_fr_mid_b_capsule, center=<9.6588 0 -1.5281>, rotation=<rot 79.2825 0 0>}]
 dyn_skirt_fr_r_a [{shapeType=Capsule, length=7, radius=5.5, center=<12.2957 -3.2806 0>, rotation=<rot 0 -9.2412 0>}]
 dyn_skirt_fr_r_b [{
   shapeType=Capsule
   length=9.9936
   radius=2.5064
   shapeName=dyn_skirt_fr_r_b_capsule
   center=<2.8992 -1.4747 -0.1527>
   rotation=<rot 2.3776 179.9996 180>
  }]
 dyn_skirt_fr_side_r_a [{shapeType=Capsule, length=7.5233, radius=2.9708, center=<11.0363 0 0>, rotation=<rot 90 166.7373 166.7375>}]
 dyn_skirt_fr_side_r_b [{
   shapeType=Capsule
   length=5
   radius=3
   shapeName=dyn_skirt_fr_side_r_b_capsule
   center=<6.541 0 1.1348>
   rotation=<rot 87.3366 0.0002 0.0004>
  }]
 pelvis [{shapeType=Capsule, length=7.1712, radius=6, center=<0 -1.4918 0>}]
 thigh_l [{shapeType=Capsule, length=19.6362, radius=8.7505, center=<-20.2563 -0.7654 0.7797>, rotation=<rot 0 88.2408 -88.7768>}]
 thigh_r [{shapeType=Capsule, length=34.4883, radius=8.6285, center=<23.6342 0.0789 -0.1598>, rotation=<rot 0 86.7557 -86.1992>}]
}
constraints [
 {bone1Name=thigh_l, bone2Name=pelvis, swing1LimitDegrees=45, swing2LimitDegrees=45, twistLimitDegrees=45}
 {bone1Name=thigh_r, bone2Name=pelvis, swing1LimitDegrees=45, swing2LimitDegrees=45, twistLimitDegrees=45}
 {bone1Name=dyn_skirt_fr_r_a, bone2Name=pelvis, swing1LimitDegrees=50, swing2LimitDegrees=5, twistLimitDegrees=0.5}
 {bone1Name=dyn_skirt_fr_side_r_a, bone2Name=pelvis, swing1LimitDegrees=5, swing2LimitDegrees=5, twistLimitDegrees=0.5}
 {bone1Name=dyn_skirt_bk_side_r_a, bone2Name=pelvis, swing1LimitDegrees=15, swing2LimitDegrees=5, twistLimitDegrees=0.5}
 {bone1Name=dyn_skirt_bk_r_a, bone2Name=pelvis, swing1LimitDegrees=20, swing2LimitDegrees=10, twistLimitDegrees=0.5}
 {bone1Name=dyn_skirt_bk_side_l_a, bone2Name=pelvis, swing1LimitDegrees=30, swing2LimitDegrees=5, twistLimitDegrees=0.5}
 {bone1Name=dyn_skirt_bk_l_a, bone2Name=pelvis, swing1LimitDegrees=20, swing2LimitDegrees=10, twistLimitDegrees=0.5}
 {bone1Name=dyn_skirt_fr_l_a, bone2Name=pelvis, swing1LimitDegrees=20, swing2LimitDegrees=10, twistLimitDegrees=0.5}
 {bone1Name=dyn_skirt_fr_mid_a, bone2Name=pelvis, swing1LimitDegrees=15, swing2LimitDegrees=5, twistLimitDegrees=0.5}
 {bone1Name=dyn_skirt_bk_l_b, bone2Name=dyn_skirt_bk_l_a, swing1LimitDegrees=60, swing2LimitDegrees=5, twistLimitDegrees=0.5}
 {bone1Name=dyn_skirt_bk_r_b, bone2Name=dyn_skirt_bk_r_a, swing1LimitDegrees=60, swing2LimitDegrees=5, twistLimitDegrees=0.5}
 {bone1Name=dyn_skirt_bk_side_l_b, bone2Name=dyn_skirt_bk_side_l_a, swing1LimitDegrees=60, swing2LimitDegrees=15, twistLimitDegrees=0.5}
 {bone1Name=dyn_skirt_bk_side_r_b, bone2Name=dyn_skirt_bk_side_r_a, swing1LimitDegrees=30, swing2LimitDegrees=30, twistLimitDegrees=0.5}
 {
  bone1Name=dyn_skirt_fr_l_b
  bone2Name=dyn_skirt_fr_l_a
  swing1LimitDegrees=15
  swing1Motion=Locked
  swing2LimitDegrees=5
  swing2Motion=Locked
  twistLimitDegrees=0.5
  twistMotion=Locked
 }
 {
  bone1Name=dyn_skirt_fr_r_b
  bone2Name=dyn_skirt_fr_r_a
  swing1LimitDegrees=30
  swing2LimitDegrees=5
  swing2Motion=Locked
  twistLimitDegrees=0.5
  twistMotion=Locked
 }
 {bone1Name=dyn_skirt_fr_side_r_b, bone2Name=dyn_skirt_fr_side_r_a, swing1LimitDegrees=5, swing2LimitDegrees=15, twistLimitDegrees=0.5}
 {
  bone1Name=calf_r
  bone2Name=thigh_r
  swing1LimitDegrees=45
  swing1Motion=Free
  swing2LimitDegrees=45
  swing2Motion=Free
  twistLimitDegrees=45
  twistMotion=Free
 }
 {
  bone1Name=dyn_skirt_fr_mid_b
  bone2Name=dyn_skirt_fr_mid_a
  swing1LimitDegrees=45
  swing1Motion=Locked
  swing2LimitDegrees=45
  swing2Motion=Locked
  twistLimitDegrees=0.5
  twistMotion=Locked
 }
]
