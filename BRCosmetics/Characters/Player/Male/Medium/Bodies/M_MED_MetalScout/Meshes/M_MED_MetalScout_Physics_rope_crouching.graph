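# Physics asset for the MetalScout male medium body's dangling ropes and amulet, named as the crouching variant.
# Front and back rope capsules each attach to a small base capsule on the pelvis, with all these joints angularly locked.
# The amulet's base is also locked to the pelvis, but the amulet itself has narrower, unlocked swing and twist limits.
# Pelvis and left thigh capsules provide body shapes, joined by an angularly free constraint.
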PhysicsAsset /BRCosmetics/Characters/Player/Male/Medium/Bodies/M_MED_MetalScout/Meshes/M_MED_MetalScout_Physics_rope_crouching
bodies {
 dyn_amulet [{
   shapeType=Capsule
   length=9.7224
   radius=4.1352
   shapeName=dyn_amulet_capsule
   center=<6.6078 -0.7028 -0.6146>
   rotation=<rot -55.8204 171.1248 125.841>
  }]
 dyn_amulet_base [{shapeType=Capsule, length=0.505, radius=0.505, shapeName=dyn_amulet_base_capsule, rotation=<rot 0 -90 0>}]
 dyn_rope_bk_a [{shapeType=Capsule, length=2, radius=1, shapeName=dyn_rope_bk_a_capsule}]
 dyn_rope_bk_b [{shapeType=Capsule, length=2, radius=1, shapeName=dyn_rope_bk_b_capsule}]
 dyn_rope_bk_base_a [{shapeType=Capsule, length=0.5, radius=0.5, shapeName=dyn_rope_bk_base_a_capsule}]
 dyn_rope_bk_base_b [{shapeType=Capsule, length=0.5, radius=0.5, shapeName=dyn_rope_bk_base_b_capsule}]
 dyn_rope_bk_base_c [{shapeType=Capsule, length=0.5, radius=0.5, shapeName=dyn_rope_bk_base_c_capsule}]
 dyn_rope_bk_base_d [{shapeType=Capsule, length=0.5, radius=0.5, shapeName=dyn_rope_bk_base_d_capsule}]
 dyn_rope_bk_c [{shapeType=Capsule, length=2, radius=1, shapeName=dyn_rope_bk_c_capsule, rotation=<rot -80 0.0001 -0.0001>}]
 dyn_rope_bk_d [{shapeType=Capsule, length=2, radius=1, shapeName=dyn_rope_bk_d_capsule, rotation=<rot 84.9999 0.0004 -14.9996>}]
 dyn_rope_fr_a [{shapeType=Capsule, length=2, radius=1, shapeName=dyn_rope_fr_a_capsule, rotation=<rot 80 -0.0001 -0.0001>}]
 dyn_rope_fr_b [{shapeType=Capsule, length=2, radius=1, shapeName=dyn_rope_fr_b_capsule, rotation=<rot 75 -0.0001 -0.0001>}]
 dyn_rope_fr_base_a [{shapeType=Capsule, length=0.5, radius=0.5, shapeName=dyn_rope_fr_base_a_capsule}]
 dyn_rope_fr_base_b [{shapeType=Capsule, length=0.5, radius=0.5, shapeName=dyn_rope_fr_base_b_capsule}]
 dyn_rope_fr_base_c [{shapeType=Capsule, length=0.5, radius=0.5, shapeName=dyn_rope_fr_base_c_capsule}]
 dyn_rope_fr_c [{shapeType=Capsule, length=2, radius=1, shapeName=dyn_rope_fr_c_capsule, rotation=<rot 70 -0.0001 -0.0001>}]
 pelvis [{
   shapeType=Capsule
   length=13.0709
   radius=7.5071
   shapeName=pelvis_capsule
   center=<-1.803 -1.2792 -1.7748>
   rotation=<rot 2.3848 -90.264 -0.7355>
  }]
 thigh_l [{
   shapeType=Capsule
   length=17.582
   radius=10.8047
   shapeName=thigh_l_capsule
   center=<-20.3598 -1.9962 3.852>
   rotation=<rot 0 88.2408 -88.7768>
  }]
}
constraints [
 {
  bone1Name=thigh_l
  bone2Name=pelvis
  swing1LimitDegrees=45
  swing1Motion=Free
  swing2LimitDegrees=45
  swing2Motion=Free
  twistLimitDegrees=45
  twistMotion=Free
 }
 {
  bone1Name=dyn_rope_fr_a
  bone2Name=dyn_rope_fr_base_a
  swing1LimitDegrees=5
  swing1Motion=Locked
  swing2LimitDegrees=5
  swing2Motion=Locked
  twistLimitDegrees=45
  twistMotion=Locked
 }
 {
  bone1Name=dyn_rope_fr_base_a
  bone2Name=pelvis
  swing1LimitDegrees=45
  swing1Motion=Locked
  swing2LimitDegrees=45
  swing2Motion=Locked
  twistLimitDegrees=45
  twistMotion=Locked
 }
 {
  bone1Name=dyn_rope_fr_base_b
  bone2Name=pelvis
  swing1LimitDegrees=45
  swing1Motion=Locked
  swing2LimitDegrees=45
  swing2Motion=Locked
  twistLimitDegrees=45
  twistMotion=Locked
 }
 {
  bone1Name=dyn_rope_fr_b
  bone2Name=dyn_rope_fr_base_b
  swing1LimitDegrees=45
  swing1Motion=Locked
  swing2LimitDegrees=45
  swing2Motion=Locked
  twistLimitDegrees=45
  twistMotion=Locked
 }
 {
  bone1Name=dyn_rope_fr_base_c
  bone2Name=pelvis
  swing1LimitDegrees=45
  swing1Motion=Locked
  swing2LimitDegrees=45
  swing2Motion=Locked
  twistLimitDegrees=45
  twistMotion=Locked
 }
 {
  bone1Name=dyn_rope_fr_c
  bone2Name=dyn_rope_fr_base_c
  swing1LimitDegrees=60
  swing1Motion=Locked
  swing2LimitDegrees=60
  swing2Motion=Locked
  twistLimitDegrees=45
  twistMotion=Locked
 }
 {
  bone1Name=dyn_rope_bk_base_a
  bone2Name=pelvis
  swing1LimitDegrees=45
  swing1Motion=Locked
  swing2LimitDegrees=45
  swing2Motion=Locked
  twistLimitDegrees=45
  twistMotion=Locked
 }
 {
  bone1Name=dyn_rope_bk_a
  bone2Name=dyn_rope_bk_base_a
  swing1LimitDegrees=45
  swing1Motion=Locked
  swing2LimitDegrees=45
  swing2Motion=Locked
  twistLimitDegrees=45
  twistMotion=Locked
 }
 {
  bone1Name=dyn_rope_bk_base_b
  bone2Name=pelvis
  swing1LimitDegrees=45
  swing1Motion=Locked
  swing2LimitDegrees=45
  swing2Motion=Locked
  twistLimitDegrees=45
  twistMotion=Locked
 }
 {
  bone1Name=dyn_rope_bk_b
  bone2Name=dyn_rope_bk_base_b
  swing1LimitDegrees=45
  swing1Motion=Locked
  swing2LimitDegrees=45
  swing2Motion=Locked
  twistLimitDegrees=45
  twistMotion=Locked
 }
 {
  bone1Name=dyn_rope_bk_base_c
  bone2Name=pelvis
  swing1LimitDegrees=45
  swing1Motion=Locked
  swing2LimitDegrees=45
  swing2Motion=Locked
  twistLimitDegrees=45
  twistMotion=Locked
 }
 {
  bone1Name=dyn_rope_bk_c
  bone2Name=dyn_rope_bk_base_c
  swing1LimitDegrees=45
  swing1Motion=Locked
  swing2LimitDegrees=45
  swing2Motion=Locked
  twistLimitDegrees=45
  twistMotion=Locked
 }
 {
  bone1Name=dyn_rope_bk_base_d
  bone2Name=pelvis
  swing1LimitDegrees=45
  swing1Motion=Locked
  swing2LimitDegrees=45
  swing2Motion=Locked
  twistLimitDegrees=45
  twistMotion=Locked
 }
 {
  bone1Name=dyn_rope_bk_d
  bone2Name=dyn_rope_bk_base_d
  swing1LimitDegrees=45
  swing1Motion=Locked
  swing2LimitDegrees=45
  swing2Motion=Locked
  twistLimitDegrees=45
  twistMotion=Locked
 }
 {bone1Name=dyn_amulet, bone2Name=dyn_amulet_base, swing1LimitDegrees=15, swing2LimitDegrees=35, twistLimitDegrees=15}
 {
  bone1Name=dyn_amulet_base
  bone2Name=pelvis
  swing1LimitDegrees=45
  swing1Motion=Locked
  swing2LimitDegrees=45
  swing2Motion=Locked
  twistLimitDegrees=45
  twistMotion=Locked
 }
]
