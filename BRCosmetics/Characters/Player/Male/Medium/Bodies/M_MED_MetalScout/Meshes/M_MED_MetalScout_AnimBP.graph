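# Animation Blueprint for the male medium MetalScout body cosmetic, copying the parent mesh pose and adding secondary motion to skirt, ropes and amulet.
# Corrective bone offsets and thigh-driven constraints lead into RigidBody skirt simulation, then a Control Rig.
# Weighted multi-bone constraints attach ropes and amulet to the skirt.
# Alpha-driven offsets adjust them for crouch, movement and sprint before another RigidBody pass, which fades by inverse crouch alpha.
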AnimBlueprint /BRCosmetics/Characters/Player/Male/Medium/Bodies/M_MED_MetalScout/Meshes/M_MED_MetalScout_AnimBP
target_skeleton=&/M_MED_MetalScout_Skeleton
graphs [
 AnimGraph {nodes [
   "Transform (Modify) Bone - Bone: dyn_rope_fr_base_b"(Alpha=1, Translation="6.758,6.626,0.713", Rotation="0.016,-17.317,-0.043", Scale="1,1,1") {
    rotationMode=BMM_Additive
    rotationSpace=BCS_BoneSpace
    translationMode=BMM_Additive
    translationSpace=BCS_BoneSpace
    alphaBoolBlend {blendOption=Linear}
    alphaScaleBiasClamp {bInterpResult=true, interpSpeedDecreasing=8, interpSpeedIncreasing=8}
    boneToModify {boneName=dyn_rope_fr_base_b}
    rotation=<rot 0.016 -17.317 -0.043>
    translation=<6.758 6.626 0.713>
   }
   "Transform (Modify) Bone - Bone: dyn_amulet_base"(Alpha=1, Translation="-2.045,0.883,0.726", Rotation="-48.006,39.577,-30.206", Scale="1,1,1") {
    rotationMode=BMM_Additive
    rotationSpace=BCS_BoneSpace
    translationMode=BMM_Additive
    translationSpace=BCS_BoneSpace
    alphaBoolBlend {blendOption=Linear}
    boneToModify {boneName=dyn_amulet_base}
    rotation=<rot -48.006 39.577 -30.206>
    translation=<-2.045 0.883 0.726>
   }
   n1 = Get Alpha_crouching() -> (Alpha_crouching:double)
   "Transform (Modify) Bone - Bone: dyn_skirt_bk_side_l_a"(Alpha=@n1, Translation="0.002,-2.305,0.871", Rotation="-2.909,15.729,-25.494", Scale="1,1,1") {
    rotationMode=BMM_Additive
    rotationSpace=BCS_BoneSpace
    translationMode=BMM_Additive
    translationSpace=BCS_BoneSpace
    alphaBoolBlend {blendOption=Linear}
    boneToModify {boneName=dyn_skirt_bk_side_l_a}
    rotation=<rot -2.909 15.729 -25.494>
    translation=<0.002 -2.305 0.871>
   }
   "Transform (Modify) Bone - Bone: dyn_rope_fr_base_c"(Alpha=1, Translation="25.705,23.226,29.769", Rotation="21.364,33.964,-5.521", Scale="1,1,1") {
    rotationMode=BMM_Additive
    rotationSpace=BCS_BoneSpace
    translationMode=BMM_Additive
    translationSpace=BCS_BoneSpace
    alphaBoolBlend {blendOption=Linear}
    alphaScaleBiasClamp {bInterpResult=true, interpSpeedDecreasing=8, interpSpeedIncreasing=8}
    boneToModify {boneName=dyn_rope_fr_base_c}
    rotation=<rot 21.364 33.964 -5.521>
    translation=<25.705 23.226 29.769>
   }
   n2 = "Transform (Modify) Bone - Bone: dyn_rope_fr_base_a"(Alpha=@n1, Translation="-0.074,1.159,-0.327", Rotation="0.129,0.105,52.937", Scale="1,1,1") {
    rotationMode=BMM_Additive
    rotationSpace=BCS_BoneSpace
    translationMode=BMM_Additive
    translationSpace=BCS_BoneSpace
    alphaBoolBlend {blendOption=Linear}
    alphaScaleBiasClamp {bInterpResult=true, interpSpeedDecreasing=8, interpSpeedIncreasing=8}
    boneToModify {boneName=dyn_rope_fr_base_a}
    rotation=<rot 0.129 0.105 52.937>
    translation=<-0.074 1.159 -0.327>
   }
   n3 = "Transform (Modify) Bone - Bone: dyn_rope_fr_base_b"(ComponentPose=@n2, Alpha=@n1, Translation="7.909,7.056,0.713", Rotation="0.016,-17.317,-0.043", Scale="1,1,1") {
    rotationMode=BMM_Additive
    rotationSpace=BCS_BoneSpace
    translationMode=BMM_Additive
    translationSpace=BCS_BoneSpace
    alphaBoolBlend {blendOption=Linear}
    alphaScaleBiasClamp {bInterpResult=true, interpSpeedDecreasing=8, interpSpeedIncreasing=8}
    boneToModify {boneName=dyn_rope_fr_base_b}
    rotation=<rot 0.016 -17.317 -0.043>
    translation=<7.909 7.056 0.713>
   }
   n4 = "Transform (Modify) Bone - Bone: dyn_rope_fr_base_c"(ComponentPose=@n3, Alpha=@n1, Translation="25.166,23.853,29.837", Rotation="21.364,33.964,-5.521", Scale="1,1,1") {
    rotationMode=BMM_Additive
    rotationSpace=BCS_BoneSpace
    translationMode=BMM_Additive
    translationSpace=BCS_BoneSpace
    alphaBoolBlend {blendOption=Linear}
    alphaScaleBiasClamp {bInterpResult=true, interpSpeedDecreasing=8, interpSpeedIncreasing=8}
    boneToModify {boneName=dyn_rope_fr_base_c}
    rotation=<rot 21.364 33.964 -5.521>
    translation=<25.1662 23.8526 29.8369>
   }
   n5 = "Transform (Modify) Bone - Bone: dyn_rope_bk_base_d"(ComponentPose=@n4, Alpha=@n1, Translation="-6.207,12.824,-1.868", Rotation="-1.059,47.467,-12.408", Scale="1,1,1") {
    rotationMode=BMM_Additive
    rotationSpace=BCS_BoneSpace
    translationMode=BMM_Additive
    translationSpace=BCS_BoneSpace
    alphaBoolBlend {blendOption=Linear}
    alphaScaleBiasClamp {bInterpResult=true, interpSpeedDecreasing=8, interpSpeedIncreasing=8}
    boneToModify {boneName=dyn_rope_bk_base_d}
    rotation=<rot -1.059 47.467 -12.408>
    translation=<-6.207 12.824 -1.868>
   }
   n6 = "Transform (Modify) Bone - Bone: dyn_amulet_base"(ComponentPose=@n5, Alpha=@n1, Translation="-10.214,-2.229,7.801", Rotation="-18.629,23.152,-35.154", Scale="1,1,1") {
    rotationMode=BMM_Additive
    rotationSpace=BCS_BoneSpace
    translationMode=BMM_Additive
    translationSpace=BCS_BoneSpace
    alphaBoolBlend {blendOption=Linear}
    boneToModify {boneName=dyn_amulet_base}
    rotation=<rot -18.629 23.152 -35.154>
    translation=<-10.214 -2.229 7.801>
   }
   "Transform (Modify) Bone - Bone: dyn_rope_bk_base_b"(ComponentPose=@n6, Alpha=@n1, Translation="0.002,-2.305,0.871", Rotation="-15.010,15.007,-0.032", Scale="1,1,1") {
    rotationMode=BMM_Additive
    rotationSpace=BCS_BoneSpace
    translationMode=BMM_Additive
    translationSpace=BCS_BoneSpace
    alphaBoolBlend {blendOption=Linear}
    boneToModify {boneName=dyn_rope_bk_base_b}
    rotation=<rot -15.01 15.007 -0.032>
    translation=<0.002 -2.305 0.871>
   }
   n7 = Get Alpha_thigh_l_rope_crouch() -> (Alpha_thigh_l_rope_crouch:double)
   "Transform (Modify) Bone - Bone: dyn_amulet_base"(Alpha=@n7, Translation="0,0,0", Rotation="-19.808,23.967,-18.945", Scale="1,1,1") {
    rotationMode=BMM_Additive
    rotationSpace=BCS_BoneSpace
    translationMode=BMM_Additive
    translationSpace=BCS_BoneSpace
    alphaBoolBlend {blendOption=Linear}
    boneToModify {boneName=dyn_amulet_base}
    rotation=<rot -19.808 23.967 -18.945>
   }
   n8 = Get Alpha_thigh_l_rope_crouch()
   n9 = "Transform (Modify) Bone - Bone: dyn_rope_fr_base_a"(Alpha=@n8, Translation="-0.032,-2.719,0.571", Rotation="0.001,0.023,16.508", Scale="1,1,1") {
    rotationMode=BMM_Additive
    rotationSpace=BCS_BoneSpace
    translationMode=BMM_Additive
    translationSpace=BCS_BoneSpace
    alphaBoolBlend {blendOption=Linear}
    alphaScaleBiasClamp {bInterpResult=true, interpSpeedDecreasing=12, interpSpeedIncreasing=3}
    boneToModify {boneName=dyn_rope_fr_base_a}
    rotation=<rot 0.001 0.023 16.508>
    translation=<-0.032 -2.719 0.571>
   }
   n10 = "Transform (Modify) Bone - Bone: dyn_rope_fr_base_b"(ComponentPose=@n9, Alpha=@n8, Translation="8.334,-2.835,-1.268", Rotation="0,0,0", Scale="1,1,1") {
    rotationMode=BMM_Additive
    rotationSpace=BCS_BoneSpace
    translationMode=BMM_Additive
    translationSpace=BCS_BoneSpace
    alphaBoolBlend {blendOption=Linear}
    alphaScaleBiasClamp {bInterpResult=true, interpSpeedDecreasing=12, interpSpeedIncreasing=3}
    boneToModify {boneName=dyn_rope_fr_base_b}
    translation=<8.334 -2.835 -1.268>
   }
   "Transform (Modify) Bone - Bone: dyn_rope_fr_base_c"(ComponentPose=@n10, Alpha=@n8, Translation="-35.813,18.249,48.245", Rotation="0,0,0", Scale="1,1,1") {
    rotationMode=BMM_Additive
    rotationSpace=BCS_BoneSpace
    translationMode=BMM_Additive
    translationSpace=BCS_BoneSpace
    alphaBoolBlend {blendOption=Linear}
    alphaScaleBiasClamp {bInterpResult=true, interpSpeedDecreasing=12, interpSpeedIncreasing=3}
    boneToModify {boneName=dyn_rope_fr_base_c}
    translation=<-35.813 18.249 48.245>
   }
   "Transform (Modify) Bone - Bone: dyn_amulet_base"(Alpha=@n8, Translation="0,0,0", Rotation="-19.808,23.967,-18.945", Scale="1,1,1") {
    rotationMode=BMM_Additive
    rotationSpace=BCS_BoneSpace
    translationMode=BMM_Additive
    translationSpace=BCS_BoneSpace
    alphaBoolBlend {blendOption=Linear}
    boneToModify {boneName=dyn_amulet_base}
    rotation=<rot -19.808 23.967 -18.945>
   }
   "Transform (Modify) Bone - Bone: dyn_amulet_base"(Alpha=1, Translation="-4.398,0.424,3.139", Rotation="-48.006,39.577,-30.206", Scale="1,1,1") {
    rotationMode=BMM_Additive
    rotationSpace=BCS_BoneSpace
    translationMode=BMM_Additive
    translationSpace=BCS_BoneSpace
    alphaBoolBlend {blendOption=Linear}
    boneToModify {boneName=dyn_amulet_base}
    rotation=<rot -48.006 39.577 -30.206>
    translation=<-4.398 0.424 3.139>
   }
   "Transform (Modify) Bone - Bone: dyn_amulet_base"(Alpha=1, Translation="-4.674,1.065,2.978", Rotation="-51.202,39.206,-21.301", Scale="1,1,1") {
    rotationMode=BMM_Additive
    rotationSpace=BCS_BoneSpace
    translationMode=BMM_Additive
    translationSpace=BCS_BoneSpace
    alphaBoolBlend {blendOption=Linear}
    boneToModify {boneName=dyn_amulet_base}
    rotation=<rot -51.202 39.206 -21.301>
    translation=<-4.674 1.065 2.978>
   }
   n11 = Get Gravity() -> (Gravity:Vector)
   n12 = Get Alpha_crouching()
   n13 = Get Alpha_crouching()
   n14 = "Transform (Modify) Bone - Bone: extra_elbowpad_l"(
    ComponentPose=Local To Component(LocalPose=Copy Pose From Mesh() {bUseAttachedParent=true})
    Alpha=Get Alpha_elbowpad_l() -> (Alpha_elbowpad_l:double)
    Translation="1.795,0.031,0.005"
    Rotation="0.266,-20.730,-0.005"
    Scale="1,1,1"
   ) {
    rotationMode=BMM_Additive
    rotationSpace=BCS_ParentBoneSpace
    translationMode=BMM_Additive
    translationSpace=BCS_ParentBoneSpace
    alphaBoolBlend {blendOption=Linear}
    alphaScaleBiasClamp {bInterpResult=true, interpSpeedDecreasing=12, interpSpeedIncreasing=12}
    boneToModify {boneName=extra_elbowpad_l}
    rotation=<rot 0.266 -20.73 -0.005>
    translation=<1.7953 0.0309 0.005>
   }
   n15 = "Constraint - extra_foot_r (RR)"!(ComponentPose=@n14, Alpha=1, ConstraintWeights_0=0.5, ConstraintWeights_1:float=0.5) -> (Pose) {
    alphaBoolBlend {blendOption=Linear}
    boneToModify {boneName=extra_foot_r}
    constraintSetup [{transformType=Rotation, targetBone {boneName=calf_r}}, {transformType=Rotation, targetBone {boneName=foot_r}}]
    constraintWeights [1,1]
   }
   n16 = "Constraint - extra_foot_l (RR)"!(ComponentPose=@n15, Alpha=1, ConstraintWeights_0=0.5, ConstraintWeights_1:float=0.5) -> (Pose) {
    alphaBoolBlend {blendOption=Linear}
    boneToModify {boneName=extra_foot_l}
    constraintSetup [{transformType=Rotation, targetBone {boneName=calf_l}}, {transformType=Rotation, targetBone {boneName=foot_l}}]
    constraintWeights [1,1]
   }
   n17 = "Transform (Modify) Bone - Bone: extra_upperarm_fr_l"(ComponentPose=@n16, Alpha=1, Translation="0.583,14.608,-6.050", Rotation="47.784,24.086,17.721", Scale="1,1,1") {
    rotationMode=BMM_Additive
    rotationSpace=BCS_BoneSpace
    translationMode=BMM_Additive
    translationSpace=BCS_BoneSpace
    alphaBoolBlend {blendOption=Linear}
    boneToModify {boneName=extra_upperarm_fr_l}
    rotation=<rot 47.784 24.086 17.721>
    translation=<0.583 14.608 -6.05>
   }
   n18 = "Transform (Modify) Bone - Bone: extra_upperarm_up_l"(ComponentPose=@n17, Alpha=1, Translation="-4.063,0.902,3.275", Rotation="59.145,92.402,118.601", Scale="1,1,1") {
    rotationMode=BMM_Additive
    rotationSpace=BCS_BoneSpace
    translationMode=BMM_Additive
    translationSpace=BCS_BoneSpace
    alphaBoolBlend {blendOption=Linear}
    boneToModify {boneName=extra_upperarm_up_l}
    rotation=<rot 59.145 92.402 118.601>
    translation=<-4.063 0.902 3.275>
   }
   n19 = "Transform (Modify) Bone - Bone: extra_upperarm_bk_l"(ComponentPose=@n18, Alpha=1, Translation="8.131,-6.987,2.516", Rotation="-29.797,-47.306,31.662", Scale="1,1,1") {
    rotationMode=BMM_Additive
    rotationSpace=BCS_BoneSpace
    translationMode=BMM_Additive
    translationSpace=BCS_BoneSpace
    alphaBoolBlend {blendOption=Linear}
    boneToModify {boneName=extra_upperarm_bk_l}
    rotation=<rot -29.797 -47.306 31.662>
    translation=<8.1312 -6.9866 2.5158>
   }
   n20 = "Constraint - dyn_skirt_fr_l_a (R)"(ComponentPose=@n19, Alpha=1, ConstraintWeights_0=1) {
    alphaBoolBlend {blendOption=Linear}
    boneToModify {boneName=dyn_skirt_fr_l_a}
    constraintSetup [{transformType=Rotation, targetBone {boneName=thigh_l}}]
    constraintWeights [1]
   }
   n21 = "Constraint - dyn_skirt_fr_mid_a (RR)"!(ComponentPose=@n20, Alpha=1, ConstraintWeights_0=0.5, ConstraintWeights_1:float=0.5) -> (Pose) {
    alphaBoolBlend {blendOption=Linear}
    boneToModify {boneName=dyn_skirt_fr_mid_a}
    constraintSetup [{transformType=Rotation, targetBone {boneName=thigh_l}}, {transformType=Rotation, targetBone {boneName=thigh_r}}]
    constraintWeights [1,1]
   }
   n22 = "Constraint - dyn_skirt_fr_r_a (R)"(ComponentPose=@n21, Alpha=0.8, ConstraintWeights_0=1) {
    alphaBoolBlend {blendOption=Linear}
    boneToModify {boneName=dyn_skirt_fr_r_a}
    constraintSetup [{transformType=Rotation, targetBone {boneName=thigh_r}}]
    constraintWeights [1]
   }
   n23 = "Constraint - dyn_skirt_fr_side_r_a (R)"(ComponentPose=@n22, Alpha=1, ConstraintWeights_0=1) {
    alphaBoolBlend {blendOption=Linear}
    boneToModify {boneName=dyn_skirt_fr_side_r_a}
    constraintSetup [{transformType=Rotation, targetBone {boneName=thigh_r}}]
    constraintWeights [1]
   }
   n24 = "Constraint - dyn_skirt_bk_side_r_a (R)"(ComponentPose=@n23, Alpha=0.4, ConstraintWeights_0=1) {
    alphaBoolBlend {blendOption=Linear}
    boneToModify {boneName=dyn_skirt_bk_side_r_a}
    constraintSetup [{transformType=Rotation, targetBone {boneName=thigh_r}}]
    constraintWeights [1]
   }
   n25 = "Transform (Modify) Bone - Bone: dyn_skirt_bk_side_l_a"(ComponentPose=@n24, Alpha=1, Translation="0,0,0", Rotation="5.838,-5.345,-14.379", Scale="1,1,1") {
    rotationMode=BMM_Additive
    rotationSpace=BCS_BoneSpace
    alphaBoolBlend {blendOption=Linear}
    boneToModify {boneName=dyn_skirt_bk_side_l_a}
    rotation=<rot 5.838 -5.345 -14.379>
   }
   n26 = "Constraint - dyn_skirt_bk_l_a (R)"(ComponentPose=@n25, Alpha=@n13, ConstraintWeights_0=1) {
    alphaBoolBlend {blendOption=Linear}
    alphaScaleBiasClamp {scale=0.3}
    boneToModify {boneName=dyn_skirt_bk_l_a}
    constraintSetup [{transformType=Rotation, targetBone {boneName=thigh_l}}]
    constraintWeights [1]
   }
   n27 = "Constraint - dyn_skirt_bk_side_l_a (R)"(ComponentPose=@n26, Alpha=@n13, ConstraintWeights_0=1) {
    alphaBoolBlend {blendOption=Linear}
    alphaScaleBiasClamp {scale=0.2}
    boneToModify {boneName=dyn_skirt_bk_side_l_a}
    constraintSetup [{transformType=Rotation, targetBone {boneName=thigh_l}}]
    constraintWeights [1]
   }
   n28 = "Constraint - dyn_skirt_bk_r_a (R)"(ComponentPose=@n27, Alpha=@n13, ConstraintWeights_0=1) {
    alphaBoolBlend {blendOption=Linear}
    alphaScaleBiasClamp {scale=0.5}
    boneToModify {boneName=dyn_skirt_bk_r_a}
    constraintSetup [{transformType=Rotation, targetBone {boneName=thigh_r}}]
    constraintWeights [1]
   }
   n29 = "Constraint - dyn_skirt_bk_r_b (R)"(ComponentPose=@n28, Alpha=@n13, ConstraintWeights_0=1) {
    alphaBoolBlend {blendOption=Linear}
    alphaScaleBiasClamp {scale=0.8}
    boneToModify {boneName=dyn_skirt_bk_r_b}
    constraintSetup [{transformType=Rotation, targetBone {boneName=thigh_r}}]
    constraintWeights [1]
   }
   n30 = RigidBody!(ComponentPose=@n29, Alpha=0.5, OverrideWorldGravity:Vector=Get Gravity(), ExternalForce) -> (Pose) {
    bOverrideWorldGravity=true
    overridePhysicsAsset=&/M_MED_MetalScout_Physics
    simulationSpace=BaseBoneSpace
    alphaBoolBlend {blendOption=Linear}
    baseBoneRef {boneName=pelvis}
    componentAppliedLinearAccClamp=<2500 2500 2500>
    componentLinearAccScale=<0.35 0.35 0.35>
    componentLinearVelScale=<0.35 0.35 0.35>
   }
   n31 = "Constraint - dyn_skirt_fr_l_a (R)"(ComponentPose=@n30, Alpha=1, ConstraintWeights_0=1) {
    alphaBoolBlend {blendOption=Linear}
    boneToModify {boneName=dyn_skirt_fr_l_a}
    constraintSetup [{transformType=Rotation, targetBone {boneName=thigh_l}}]
    constraintWeights [1]
   }
   n32 = "Transform (Modify) Bone - Bone: dyn_skirt_fr_l_a"(ComponentPose=@n31, Alpha=1, Translation="0.054,1.746,0.451", Rotation="6.229,25.417,-4.901", Scale="1,1,1") {
    rotationMode=BMM_Additive
    rotationSpace=BCS_BoneSpace
    scaleMode=BMM_Additive
    scaleSpace=BCS_BoneSpace
    translationMode=BMM_Additive
    translationSpace=BCS_BoneSpace
    alphaBoolBlend {blendOption=Linear}
    alphaScaleBiasClamp {bInterpResult=true, interpSpeedDecreasing=8, interpSpeedIncreasing=8}
    boneToModify {boneName=dyn_skirt_fr_l_a}
    rotation=<rot 6.229 25.417 -4.901>
    translation=<0.054 1.746 0.451>
   }
   n33 = Control Rig!(
    Source=Component To Local(ComponentPose=@n32)
    Alpha
    Alpha_skydiving:double=Get Alpha_skydiving() -> (Alpha_skydiving:double)
    Alpha_frontend:double=Get Alpha_frontend() -> (Alpha_frontend:double)
    Alpha_crouching:double=Get Alpha_crouching()
   ) -> (Pose) {
    alphaBoolBlend {blendOption=Linear}
    controlRigAssetReference {blueprintRigClass=&/M_MED_MetalScout_CtrlRig.M_MED_MetalScout_CtrlRig_C}
   }
   n34 = "Transform (Modify) Bone - Bone: dyn_skirt_bk_side_r_c"(
    ComponentPose=Local To Component(LocalPose=@n33)
    Alpha=Get Alpha_thigh_r_skirt() -> (Alpha_thigh_r_skirt:double)
    Translation="0,0,0"
    Rotation="62.635,115.840,48.086"
    Scale="1,1,1"
   ) {
    rotationMode=BMM_Additive
    rotationSpace=BCS_BoneSpace
    scaleMode=BMM_Additive
    scaleSpace=BCS_BoneSpace
    translationMode=BMM_Additive
    translationSpace=BCS_BoneSpace
    alphaBoolBlend {blendOption=Linear}
    boneToModify {boneName=dyn_skirt_bk_side_r_c}
    rotation=<rot 62.635 115.84 48.086>
   }
   n35 = "Constraint - dyn_skirt_fr_mid_a (RR)"(ComponentPose=@n34, Alpha=1, ConstraintWeights_0=1, ConstraintWeights_1=1) {
    alphaBoolBlend {blendOption=Linear}
    boneToModify {boneName=dyn_skirt_fr_mid_a}
    constraintSetup [{transformType=Rotation, targetBone {boneName=thigh_r}}, {transformType=Rotation, targetBone {boneName=thigh_l}}]
    constraintWeights [1,1]
   }
   n36 = "Transform (Modify) Bone - Bone: dyn_skirt_fr_l_a"(ComponentPose=@n35, Alpha=Get Alpha_crouching(), Translation="0,0,0", Rotation="26.605,-25.436,9.023", Scale="0.8,1,1") {
    rotationMode=BMM_Additive
    rotationSpace=BCS_BoneSpace
    scaleMode=BMM_Additive
    scaleSpace=BCS_BoneSpace
    translationMode=BMM_Additive
    translationSpace=BCS_BoneSpace
    alphaBoolBlend {blendOption=Linear}
    alphaScaleBiasClamp {bInterpResult=true, interpSpeedDecreasing=8, interpSpeedIncreasing=8}
    boneToModify {boneName=dyn_skirt_fr_l_a}
    rotation=<rot 26.605 -25.436 9.023>
    scale {x=0.8}
   }
   n37 = "Transform (Modify) Bone - Bone: extra_skirt"(ComponentPose=@n36, Alpha=1, Translation="-9.342,6.338,-53.821", Rotation="0,0,0", Scale="1,1,1") {
    rotationMode=BMM_Additive
    rotationSpace=BCS_BoneSpace
    scaleMode=BMM_Additive
    scaleSpace=BCS_BoneSpace
    translationMode=BMM_Additive
    translationSpace=BCS_BoneSpace
    alphaBoolBlend {blendOption=Linear}
    alphaScaleBiasClamp {bInterpResult=true, interpSpeedDecreasing=8, interpSpeedIncreasing=8}
    boneToModify {boneName=extra_skirt}
    translation=<-9.342 6.338 -53.821>
   }
   n38 = "Transform (Modify) Bone - Bone: extra_thighs"(ComponentPose=@n37, Alpha=1, Translation="0,0,0", Rotation="0,0,0", Scale="0.2,0.2,0.2") {
    rotationMode=BMM_Additive
    rotationSpace=BCS_BoneSpace
    scaleMode=BMM_Replace
    scaleSpace=BCS_BoneSpace
    translationMode=BMM_Additive
    translationSpace=BCS_BoneSpace
    alphaBoolBlend {blendOption=Linear}
    boneToModify {boneName=extra_thighs}
    scale=<0.2 0.2 0.2>
   }
   n39 = "Transform (Modify) Bone - Bone: extra_thigh_bk_r"(ComponentPose=@n38, Alpha=1, Translation="5.114,-0.434,-1.422", Rotation="0,0,0", Scale="0.9,0.9,0.99") {
    rotationMode=BMM_Additive
    rotationSpace=BCS_BoneSpace
    scaleMode=BMM_Replace
    scaleSpace=BCS_BoneSpace
    translationMode=BMM_Additive
    translationSpace=BCS_BoneSpace
    alphaBoolBlend {blendOption=Linear}
    boneToModify {boneName=extra_thigh_bk_r}
    scale=<0.9 0.9 0.99>
    translation=<5.114 -0.434 -1.422>
   }
   n40 = "Transform (Modify) Bone - Bone: extra_thigh_bk_l"(ComponentPose=@n39, Alpha=1, Translation="0,0,0", Rotation="0,0,0", Scale="0.95,0.95,0.95") {
    rotationMode=BMM_Additive
    rotationSpace=BCS_BoneSpace
    scaleMode=BMM_Replace
    scaleSpace=BCS_BoneSpace
    translationMode=BMM_Additive
    translationSpace=BCS_BoneSpace
    alphaBoolBlend {blendOption=Linear}
    boneToModify {boneName=extra_thigh_bk_l}
    scale=<0.95 0.95 0.95>
   }
   n41 = "Constraint - dyn_rope_fr_base_a (PPPPPPPP)"!(
    ComponentPose=@n40
    Alpha=1
    ConstraintWeights_0=0.0067
    ConstraintWeights_1:float=0.0005
    ConstraintWeights_2:float=0.0003
    ConstraintWeights_3:float=0.5499
    ConstraintWeights_4:float=0.1489
    ConstraintWeights_5:float=0.1056
    ConstraintWeights_6:float=0.1688
    ConstraintWeights_7:float=0.0193
   ) -> (Pose) {
    alphaBoolBlend {blendOption=Linear}
    boneToModify {boneName=dyn_rope_fr_base_a}
    constraintSetup [
     {targetBone {boneName=dyn_skirt_bk_side_l_a}}
     {targetBone {boneName=dyn_skirt_bk_side_l_b}}
     {targetBone {boneName=dyn_skirt_bk_side_r_a}}
     {targetBone {boneName=dyn_skirt_fr_r_a}}
     {targetBone {boneName=dyn_skirt_fr_mid_a}}
     {targetBone {boneName=dyn_skirt_fr_mid_b}}
     {targetBone {boneName=dyn_skirt_fr_l_b}}
     {targetBone {boneName=dyn_skirt_fr_l_a}}
    ]
    constraintWeights [1,1,1,1,1,1,1,1]
   }
   n42 = "Constraint - dyn_rope_fr_base_b (PPPPPPPP)"!(
    ComponentPose=@n41
    Alpha=1
    ConstraintWeights_0=0.041
    ConstraintWeights_1:float=0.0036
    ConstraintWeights_2:float=0.0489
    ConstraintWeights_3:float=0.0222
    ConstraintWeights_4:float=0.1691
    ConstraintWeights_5:float=0.1105
    ConstraintWeights_6:float=0.4551
    ConstraintWeights_7:float=0.1497
   ) -> (Pose) {
    alphaBoolBlend {blendOption=Linear}
    boneToModify {boneName=dyn_rope_fr_base_b}
    constraintSetup [
     {targetBone {boneName=dyn_skirt_bk_side_l_a}}
     {targetBone {boneName=dyn_skirt_bk_side_l_b}}
     {targetBone {boneName=dyn_skirt_fr_r_a}}
     {targetBone {boneName=dyn_skirt_fr_r_b}}
     {targetBone {boneName=dyn_skirt_fr_mid_a}}
     {targetBone {boneName=dyn_skirt_fr_mid_b}}
     {targetBone {boneName=dyn_skirt_fr_l_b}}
     {targetBone {boneName=dyn_skirt_fr_l_a}}
    ]
    constraintWeights [1,1,1,1,1,1,1,1]
   }
   n43 = "Constraint - dyn_rope_fr_base_c (PPPPPPPP)"!(
    ComponentPose=@n42
    Alpha=1
    ConstraintWeights_0=0.1223
    ConstraintWeights_1:float=0.016
    ConstraintWeights_2:float=0.0086
    ConstraintWeights_3:float=0.0108
    ConstraintWeights_4:float=0.0293
    ConstraintWeights_5:float=0.0306
    ConstraintWeights_6:float=0.6432
    ConstraintWeights_7:float=0.1392
   ) -> (Pose) {
    alphaBoolBlend {blendOption=Linear}
    boneToModify {boneName=dyn_rope_fr_base_c}
    constraintSetup [
     {targetBone {boneName=dyn_skirt_bk_side_l_a}}
     {targetBone {boneName=dyn_skirt_bk_side_l_b}}
     {targetBone {boneName=dyn_skirt_fr_r_a}}
     {targetBone {boneName=dyn_skirt_fr_r_b}}
     {targetBone {boneName=dyn_skirt_fr_mid_a}}
     {targetBone {boneName=dyn_skirt_fr_mid_b}}
     {targetBone {boneName=dyn_skirt_fr_l_b}}
     {targetBone {boneName=dyn_skirt_fr_l_a}}
    ]
    constraintWeights [1,1,1,1,1,1,1,1]
   }
   n44 = "Constraint - dyn_rope_bk_base_a (PPPPPPPP)"!(
    ComponentPose=@n43
    Alpha=1
    ConstraintWeights_0=0.0226
    ConstraintWeights_1:float=0.013
    ConstraintWeights_2:float=0.6633
    ConstraintWeights_3:float=0.2921
    ConstraintWeights_4:float=0.0075
    ConstraintWeights_5:float=0.0008
    ConstraintWeights_6:float=0.0004
    ConstraintWeights_7:float=0.0002
   ) -> (Pose) {
    alphaBoolBlend {blendOption=Linear}
    boneToModify {boneName=dyn_rope_bk_base_a}
    constraintSetup [
     {targetBone {boneName=dyn_skirt_bk_side_l_a}}
     {targetBone {boneName=dyn_skirt_bk_side_l_b}}
     {targetBone {boneName=dyn_skirt_bk_l_a}}
     {targetBone {boneName=dyn_skirt_bk_l_b}}
     {targetBone {boneName=dyn_skirt_bk_r_a}}
     {targetBone {boneName=dyn_skirt_bk_r_b}}
     {targetBone {boneName=dyn_skirt_bk_side_r_b}}
     {targetBone {boneName=dyn_skirt_fr_side_r_a}}
    ]
    constraintWeights [1,1,1,1,1,1,1,1]
   }
   n45 = "Constraint - dyn_rope_bk_base_b (PPPPPPPP)"!(
    ComponentPose=@n44
    Alpha=1
    ConstraintWeights_0=0.464
    ConstraintWeights_1:float=0.2142
    ConstraintWeights_2:float=0.1731
    ConstraintWeights_3:float=0.1469
    ConstraintWeights_4:float=0.0011
    ConstraintWeights_5:float=0.0002
    ConstraintWeights_6:float=0.0003
    ConstraintWeights_7:float=0.0002
   ) -> (Pose) {
    alphaBoolBlend {blendOption=Linear}
    boneToModify {boneName=dyn_rope_bk_base_b}
    constraintSetup [
     {targetBone {boneName=dyn_skirt_bk_side_l_a}}
     {targetBone {boneName=dyn_skirt_bk_side_l_b}}
     {targetBone {boneName=dyn_skirt_bk_l_a}}
     {targetBone {boneName=dyn_skirt_bk_l_b}}
     {targetBone {boneName=dyn_skirt_bk_r_b}}
     {targetBone {boneName=dyn_skirt_bk_side_r_b}}
     {targetBone {boneName=dyn_skirt_fr_l_b}}
     {targetBone {boneName=dyn_skirt_fr_l_a}}
    ]
    constraintWeights [1,1,1,1,1,1,1,1]
   }
   n46 = "Constraint - dyn_rope_bk_base_c (PPPPPPPP)"!(
    ComponentPose=@n45
    Alpha=1
    ConstraintWeights_0=0.6798
    ConstraintWeights_1:float=0.317
    ConstraintWeights_2:float=0.0003
    ConstraintWeights_3:float=0.0002
    ConstraintWeights_4:float=0.0006
    ConstraintWeights_5:float=0.0004
    ConstraintWeights_6:float=0.0011
    ConstraintWeights_7:float=0.0008
   ) -> (Pose) {
    alphaBoolBlend {blendOption=Linear}
    boneToModify {boneName=dyn_rope_bk_base_c}
    constraintSetup [
     {targetBone {boneName=dyn_skirt_bk_side_l_a}}
     {targetBone {boneName=dyn_skirt_bk_side_l_b}}
     {targetBone {boneName=dyn_skirt_bk_l_a}}
     {targetBone {boneName=dyn_skirt_bk_l_b}}
     {targetBone {boneName=dyn_skirt_bk_r_a}}
     {targetBone {boneName=dyn_skirt_fr_mid_a}}
     {targetBone {boneName=dyn_skirt_fr_l_b}}
     {targetBone {boneName=dyn_skirt_fr_l_a}}
    ]
    constraintWeights [1,1,1,1,1,1,1,1]
   }
   n47 = "Constraint - dyn_rope_bk_base_d (PPPPPPPP)"!(
    ComponentPose=@n46
    Alpha=1
    ConstraintWeights_0=0.5897
    ConstraintWeights_1:float=0.4022
    ConstraintWeights_2:float=0
    ConstraintWeights_3:float=0
    ConstraintWeights_4:float=0.0039
    ConstraintWeights_5:float=0
    ConstraintWeights_6:float=0.002
    ConstraintWeights_7:float=0.0022
   ) -> (Pose) {
    alphaBoolBlend {blendOption=Linear}
    boneToModify {boneName=dyn_rope_bk_base_d}
    constraintSetup [
     {targetBone {boneName=dyn_skirt_bk_side_l_a}}
     {targetBone {boneName=dyn_skirt_bk_side_l_b}}
     {targetBone {boneName=dyn_skirt_bk_r_a}}
     {targetBone {boneName=dyn_skirt_fr_r_b}}
     {targetBone {boneName=dyn_skirt_fr_mid_a}}
     {targetBone {boneName=dyn_skirt_fr_mid_b}}
     {targetBone {boneName=dyn_skirt_fr_l_b}}
     {targetBone {boneName=dyn_skirt_fr_l_a}}
    ]
    constraintWeights [1,1,1,1,1,1,1,1]
   }
   n48 = "Constraint - dyn_amulet_base (PPPPP)"!(
    ComponentPose=@n47
    Alpha=1
    ConstraintWeights_0=0.716
    ConstraintWeights_1:float=0.1107
    ConstraintWeights_2:float=0.0038
    ConstraintWeights_3:float=0.0373
    ConstraintWeights_4:float=0.1322
   ) -> (Pose) {
    alphaBoolBlend {blendOption=Linear}
    boneToModify {boneName=dyn_amulet_base}
    constraintSetup [
     {targetBone {boneName=dyn_skirt_bk_side_l_a}}
     {targetBone {boneName=dyn_skirt_bk_side_l_b}}
     {targetBone {boneName=dyn_skirt_fr_mid_a}}
     {targetBone {boneName=dyn_skirt_fr_l_b}}
     {targetBone {boneName=dyn_skirt_fr_l_a}}
    ]
    constraintWeights [1,1,1,1,1]
   }
   n49 = "Transform (Modify) Bone - Bone: dyn_rope_bk_base_a"!(ComponentPose=@n48, bAlphaBoolEnabled:bool=True, Translation="0.000,0.651,-0.232", Rotation="0,0,0", Scale="1,1,1") -> (Pose) {
    alphaInputType=Bool
    rotationMode=BMM_Additive
    rotationSpace=BCS_BoneSpace
    translationMode=BMM_Additive
    translationSpace=BCS_BoneSpace
    alphaBoolBlend {blendOption=Linear}
    alphaScaleBiasClamp {bInterpResult=true, interpSpeedDecreasing=1, interpSpeedIncreasing=12}
    boneToModify {boneName=dyn_rope_bk_base_a}
    translation {y=0.651, z=-0.232}
   }
   n50 = "Transform (Modify) Bone - Bone: dyn_rope_bk_base_b"!(ComponentPose=@n49, bAlphaBoolEnabled:bool=True, Translation="-0.001,1.326,-0.142", Rotation="0,0,0", Scale="1,1,1") -> (Pose) {
    alphaInputType=Bool
    rotationMode=BMM_Additive
    rotationSpace=BCS_BoneSpace
    translationMode=BMM_Additive
    translationSpace=BCS_BoneSpace
    alphaBoolBlend {blendOption=Linear}
    boneToModify {boneName=dyn_rope_bk_base_b}
    translation=<-0.001 1.326 -0.142>
   }
   n51 = "Transform (Modify) Bone - Bone: dyn_rope_bk_base_c"!(ComponentPose=@n50, bAlphaBoolEnabled:bool=True, Translation="2.867,0,-0.963", Rotation="0,0,0", Scale="1,1,1") -> (Pose) {
    alphaInputType=Bool
    rotationMode=BMM_Additive
    rotationSpace=BCS_BoneSpace
    translationMode=BMM_Additive
    translationSpace=BCS_BoneSpace
    alphaBoolBlend {blendOption=Linear}
    boneToModify {boneName=dyn_rope_bk_base_c}
    translation {x=2.867, z=-0.963}
   }
   n52 = "Transform (Modify) Bone - Bone: dyn_rope_fr_base_b"!(ComponentPose=@n51, bAlphaBoolEnabled:bool=True, Translation="-1.930,-0.435,1.344", Rotation="0.000,5.000,-0.000", Scale="1,1,1") -> (Pose) {
    alphaInputType=Bool
    rotationMode=BMM_Additive
    rotationSpace=BCS_BoneSpace
    translationMode=BMM_Additive
    translationSpace=BCS_BoneSpace
    alphaBoolBlend {blendOption=Linear}
    alphaScaleBiasClamp {bInterpResult=true, interpSpeedDecreasing=1, interpSpeedIncreasing=12}
    boneToModify {boneName=dyn_rope_fr_base_b}
    rotation=<rot 0 5 -0.0001>
    translation=<-1.93 -0.435 1.344>
   }
   n53 = "Transform (Modify) Bone - Bone: dyn_rope_fr_base_c"!(ComponentPose=@n52, bAlphaBoolEnabled:bool=True, Translation="4.586,1.959,-5.360", Rotation="-5.010,9.857,0.862", Scale="1,1,1") -> (Pose) {
    alphaInputType=Bool
    rotationMode=BMM_Additive
    rotationSpace=BCS_BoneSpace
    translationMode=BMM_Additive
    translationSpace=BCS_BoneSpace
    alphaBoolBlend {blendOption=Linear}
    boneToModify {boneName=dyn_rope_fr_base_c}
    rotation=<rot -5.01 9.857 0.862>
    translation=<4.586 1.959 -5.36>
   }
   n54 = "Transform (Modify) Bone - Bone: dyn_amulet_base"!(ComponentPose=@n53, bAlphaBoolEnabled:bool=True, Translation="-4.162,3.855,-2.445", Rotation="-25.426,2.799,-2.343", Scale="1,1,1") -> (Pose) {
    alphaInputType=Bool
    rotationMode=BMM_Additive
    rotationSpace=BCS_BoneSpace
    translationMode=BMM_Additive
    translationSpace=BCS_BoneSpace
    alphaBoolBlend {blendOption=Linear}
    boneToModify {boneName=dyn_amulet_base}
    rotation=<rot -25.426 2.799 -2.343>
    translation=<-4.162 3.855 -2.445>
   }
   n55 = "Transform (Modify) Bone - Bone: dyn_rope_bk_base_d"!(ComponentPose=@n54, bAlphaBoolEnabled:bool=True, Translation="-4.724,5.625,-4.318", Rotation="-17.245,35.031,-4.519", Scale="1,1,1") -> (Pose) {
    alphaInputType=Bool
    rotationMode=BMM_Additive
    rotationSpace=BCS_BoneSpace
    translationMode=BMM_Additive
    translationSpace=BCS_BoneSpace
    alphaBoolBlend {blendOption=Linear}
    boneToModify {boneName=dyn_rope_bk_base_d}
    rotation=<rot -17.245 35.031 -4.519>
    translation=<-4.724 5.625 -4.318>
   }
   n56 = Get Alpha_move_rope() -> (Alpha_move_rope:double)
   n57 = "Transform (Modify) Bone - Bone: dyn_rope_fr_base_a"(ComponentPose=@n55, Alpha=@n56, Translation="0.211,4.022,2.240", Rotation="0.000,0.000,31.553", Scale="1,1,1") {
    rotationMode=BMM_Additive
    rotationSpace=BCS_BoneSpace
    translationMode=BMM_Additive
    translationSpace=BCS_BoneSpace
    alphaBoolBlend {blendOption=Linear}
    alphaScaleBiasClamp {bInterpResult=true, interpSpeedDecreasing=12, interpSpeedIncreasing=3}
    boneToModify {boneName=dyn_rope_fr_base_a}
    rotation=<rot 0 0 31.553>
    translation=<0.211 4.022 2.24>
   }
   n58 = "Transform (Modify) Bone - Bone: dyn_rope_fr_base_b"!(ComponentPose=@n57, Alpha:float=@n56, Translation="1.639,-0.235,2.990", Rotation="0.030,1.109,0.005", Scale="1,1,1") -> (Pose) {
    rotationMode=BMM_Additive
    rotationSpace=BCS_BoneSpace
    translationMode=BMM_Additive
    translationSpace=BCS_BoneSpace
    alphaBoolBlend {blendOption=Linear}
    alphaScaleBiasClamp {bInterpResult=true, interpSpeedDecreasing=8, interpSpeedIncreasing=3}
    boneToModify {boneName=dyn_rope_fr_base_b}
    rotation=<rot 0.03 1.109 0.005>
    translation=<1.639 -0.235 2.99>
   }
   n59 = "Transform (Modify) Bone - Bone: dyn_rope_fr_base_c"!(ComponentPose=@n58, Alpha:float=@n56, Translation="-0.932,0.406,0.821", Rotation="12.670,-15.404,36.603", Scale="1,1,1") -> (Pose) {
    rotationMode=BMM_Additive
    rotationSpace=BCS_BoneSpace
    translationMode=BMM_Additive
    translationSpace=BCS_BoneSpace
    alphaBoolBlend {blendOption=Linear}
    alphaScaleBiasClamp {bInterpResult=true, interpSpeedDecreasing=12, interpSpeedIncreasing=3}
    boneToModify {boneName=dyn_rope_fr_base_c}
    rotation=<rot 12.67 -15.404 36.603>
    translation=<-0.932 0.406 0.821>
   }
   n60 = "Transform (Modify) Bone - Bone: dyn_rope_bk_base_d"!(ComponentPose=@n59, Alpha:float=@n56, Translation="0,3,0", Rotation="-0.000,10.000,0.000", Scale="1,1,1") -> (Pose) {
    rotationMode=BMM_Additive
    rotationSpace=BCS_BoneSpace
    translationMode=BMM_Additive
    translationSpace=BCS_BoneSpace
    alphaBoolBlend {blendOption=Linear}
    alphaScaleBiasClamp {bInterpResult=true, interpSpeedDecreasing=8, interpSpeedIncreasing=8}
    boneToModify {boneName=dyn_rope_bk_base_d}
    rotation=<rot -0.0001 10 0>
    translation {y=3}
   }
   n61 = "Transform (Modify) Bone - Bone: dyn_rope_fr_base_b"(
    ComponentPose=@n60
    Alpha=Get Alpha_sprint_rope() -> (Alpha_sprint_rope:double)
    Translation="11.951,3.263,2.397"
    Rotation="0.016,-17.317,-0.043"
    Scale="1,1,1"
   ) {
    rotationMode=BMM_Additive
    rotationSpace=BCS_BoneSpace
    translationMode=BMM_Additive
    translationSpace=BCS_BoneSpace
    alphaBoolBlend {blendOption=Linear}
    alphaScaleBiasClamp {bInterpResult=true, interpSpeedDecreasing=8, interpSpeedIncreasing=8}
    boneToModify {boneName=dyn_rope_fr_base_b}
    rotation=<rot 0.016 -17.317 -0.043>
    translation=<11.951 3.263 2.397>
   }
   n62 = "Transform (Modify) Bone - Bone: dyn_rope_fr_base_a"(ComponentPose=@n61, Alpha=@n12, Translation="-0.074,1.159,-0.327", Rotation="0.121,0.105,49.305", Scale="1,1,1") {
    rotationMode=BMM_Additive
    rotationSpace=BCS_BoneSpace
    translationMode=BMM_Additive
    translationSpace=BCS_BoneSpace
    alphaBoolBlend {blendOption=Linear}
    alphaScaleBiasClamp {bInterpResult=true, interpSpeedDecreasing=8, interpSpeedIncreasing=8}
    boneToModify {boneName=dyn_rope_fr_base_a}
    rotation=<rot 0.121 0.105 49.305>
    translation=<-0.074 1.159 -0.327>
   }
   n63 = "Transform (Modify) Bone - Bone: dyn_rope_fr_base_b"(ComponentPose=@n62, Alpha=@n12, Translation="5.052,7.253,0.658", Rotation="0.016,-17.317,-0.043", Scale="1,1,1") {
    rotationMode=BMM_Additive
    rotationSpace=BCS_BoneSpace
    translationMode=BMM_Additive
    translationSpace=BCS_BoneSpace
    alphaBoolBlend {blendOption=Linear}
    alphaScaleBiasClamp {bInterpResult=true, interpSpeedDecreasing=8, interpSpeedIncreasing=8}
    boneToModify {boneName=dyn_rope_fr_base_b}
    rotation=<rot 0.016 -17.317 -0.043>
    translation=<5.052 7.253 0.658>
   }
   n64 = "Transform (Modify) Bone - Bone: dyn_rope_fr_base_c"(ComponentPose=@n63, Alpha=@n12, Translation="25.166,23.853,29.837", Rotation="21.364,33.964,-5.521", Scale="1,1,1") {
    rotationMode=BMM_Additive
    rotationSpace=BCS_BoneSpace
    translationMode=BMM_Additive
    translationSpace=BCS_BoneSpace
    alphaBoolBlend {blendOption=Linear}
    alphaScaleBiasClamp {bInterpResult=true, interpSpeedDecreasing=8, interpSpeedIncreasing=8}
    boneToModify {boneName=dyn_rope_fr_base_c}
    rotation=<rot 21.364 33.964 -5.521>
    translation=<25.1662 23.8527 29.8369>
   }
   n65 = "Transform (Modify) Bone - Bone: dyn_rope_bk_base_d"(ComponentPose=@n64, Alpha=@n12, Translation="-6.207,12.824,-1.868", Rotation="-1.059,47.467,-12.408", Scale="1,1,1") {
    rotationMode=BMM_Additive
    rotationSpace=BCS_BoneSpace
    translationMode=BMM_Additive
    translationSpace=BCS_BoneSpace
    alphaBoolBlend {blendOption=Linear}
    alphaScaleBiasClamp {bInterpResult=true, interpSpeedDecreasing=8, interpSpeedIncreasing=8}
    boneToModify {boneName=dyn_rope_bk_base_d}
    rotation=<rot -1.059 47.467 -12.408>
    translation=<-6.207 12.824 -1.868>
   }
   n66 = "Transform (Modify) Bone - Bone: dyn_amulet_base"!(ComponentPose=@n65, Alpha:float=@n12, Translation="-10.214,-2.229,7.801", Rotation="-18.629,23.152,-35.154", Scale="1,1,1") -> (Pose) {
    rotationMode=BMM_Additive
    rotationSpace=BCS_BoneSpace
    translationMode=BMM_Additive
    translationSpace=BCS_BoneSpace
    alphaBoolBlend {blendOption=Linear}
    boneToModify {boneName=dyn_amulet_base}
    rotation=<rot -18.629 23.152 -35.154>
    translation=<-10.214 -2.229 7.801>
   }
   n67 = "Transform (Modify) Bone - Bone: dyn_rope_bk_base_b"!(ComponentPose=@n66, Alpha:float=@n12, Translation="0.002,-2.305,0.871", Rotation="-15.010,15.007,-0.032", Scale="1,1,1") -> (Pose) {
    rotationMode=BMM_Additive
    rotationSpace=BCS_BoneSpace
    translationMode=BMM_Additive
    translationSpace=BCS_BoneSpace
    alphaBoolBlend {blendOption=Linear}
    boneToModify {boneName=dyn_rope_bk_base_b}
    rotation=<rot -15.01 15.007 -0.032>
    translation=<0.002 -2.305 0.871>
   }
   n68 = "Transform (Modify) Bone - Bone: dyn_rope_fr_base_a"(ComponentPose=@n67, Alpha=@n7, Translation="-0.032,-2.719,0.571", Rotation="-0.065,0.043,67.967", Scale="1,1,1") {
    rotationMode=BMM_Additive
    rotationSpace=BCS_BoneSpace
    translationMode=BMM_Additive
    translationSpace=BCS_BoneSpace
    alphaBoolBlend {blendOption=Linear}
    alphaScaleBiasClamp {bInterpResult=true, interpSpeedDecreasing=12, interpSpeedIncreasing=3}
    boneToModify {boneName=dyn_rope_fr_base_a}
    rotation=<rot -0.065 0.043 67.967>
    translation=<-0.032 -2.719 0.571>
   }
   n69 = "Transform (Modify) Bone - Bone: dyn_rope_fr_base_b"(ComponentPose=@n68, Alpha=@n7, Translation="8.334,-3.741,-1.268", Rotation="1.313,-77.708,0.702", Scale="1,1,1") {
    rotationMode=BMM_Additive
    rotationSpace=BCS_BoneSpace
    translationMode=BMM_Additive
    translationSpace=BCS_BoneSpace
    alphaBoolBlend {blendOption=Linear}
    alphaScaleBiasClamp {bInterpResult=true, interpSpeedDecreasing=12, interpSpeedIncreasing=3}
    boneToModify {boneName=dyn_rope_fr_base_b}
    rotation=<rot 1.313 -77.708 0.702>
    translation=<8.334 -3.741 -1.268>
   }
   n70 = "Transform (Modify) Bone - Bone: dyn_rope_fr_base_c"(ComponentPose=@n69, Alpha=@n7, Translation="-35.813,17.900,46.453", Rotation="1.238,-25.559,-2.033", Scale="1,1,1") {
    rotationMode=BMM_Additive
    rotationSpace=BCS_BoneSpace
    translationMode=BMM_Additive
    translationSpace=BCS_BoneSpace
    alphaBoolBlend {blendOption=Linear}
    alphaScaleBiasClamp {bInterpResult=true, interpSpeedDecreasing=12, interpSpeedIncreasing=3}
    boneToModify {boneName=dyn_rope_fr_base_c}
    rotation=<rot 1.238 -25.559 -2.033>
    translation=<-35.813 17.9 46.453>
   }
   n71 = RigidBody(ComponentPose=@n70, Alpha=Get inv_Alpha_crouching() -> (inv_Alpha_crouching:double), OverrideWorldGravity=@n11) {
    bOverrideWorldGravity=true
    overridePhysicsAsset=&/M_MED_MetalScout_Physics_rope
    simulationSpace=BaseBoneSpace
    alphaBoolBlend {blendOption=Linear}
    baseBoneRef {boneName=dyn_amulet_base}
    componentAppliedLinearAccClamp=<2500 2500 2500>
    componentLinearAccScale=<0.35 0.35 0.35>
    componentLinearVelScale=<0.35 0.35 0.35>
    simSpaceSettings {maxAngularAcceleration=10, maxAngularVelocity=10, maxLinearAcceleration=10, maxLinearVelocity=10}
   }
   n72 = RigidBody(ComponentPose=@n71, Alpha=Get Alpha_crouching(), OverrideWorldGravity=@n11) {
    bOverrideWorldGravity=true
    overridePhysicsAsset=&/M_MED_MetalScout_Physics_rope_crouching
    simulationSpace=BaseBoneSpace
    alphaBoolBlend {blendOption=Linear}
    baseBoneRef {boneName=dyn_amulet_base}
    componentAppliedLinearAccClamp=<2500 2500 2500>
    componentLinearAccScale=<0.35 0.35 0.35>
    componentLinearVelScale=<0.35 0.35 0.35>
    simSpaceSettings {maxAngularAcceleration=10, maxAngularVelocity=10, maxLinearAcceleration=10, maxLinearVelocity=10}
   }
   n73 = "Transform (Modify) Bone - Bone: dyn_amulet_base"!(ComponentPose=@n72, bAlphaBoolEnabled:bool=True, Translation="0.182,-2.960,0.651", Rotation="1.564,9.816,-11.344", Scale="1,1,1") -> (Pose) {
    alphaInputType=Bool
    rotationMode=BMM_Additive
    rotationSpace=BCS_BoneSpace
    translationMode=BMM_Additive
    translationSpace=BCS_BoneSpace
    alphaBoolBlend {blendOption=Linear}
    boneToModify {boneName=dyn_amulet_base}
    rotation=<rot 1.564 9.816 -11.344>
    translation=<0.182 -2.96 0.651>
   }
   n74 = "Transform (Modify) Bone - Bone: extra_thigh_scale_r"!(
    ComponentPose=@n73
    bAlphaBoolEnabled:bool=Get LOD_2_3() -> (LOD_2_3:bool)
    Translation="0.004,2.653,3.421"
    Rotation="0,0,0"
    Scale="1,1.5,1.5"
   ) -> (Pose) {
    alphaInputType=Bool
    rotationMode=BMM_Additive
    rotationSpace=BCS_BoneSpace
    scaleMode=BMM_Additive
    scaleSpace=BCS_BoneSpace
    translationMode=BMM_Additive
    translationSpace=BCS_BoneSpace
    alphaBoolBlend {blendOption=Linear}
    boneToModify {boneName=extra_thigh_scale_r}
    scale {y=1.5, z=1.5}
    translation=<0.004 2.653 3.421>
   }
   n75 = "Constraint - dyn_skirt_bk_side_l_a (R)"(ComponentPose=@n74, Alpha=Get inv_Alpha_crouching(), ConstraintWeights_0=1) {
    alphaBoolBlend {blendOption=Linear}
    alphaScaleBiasClamp {scale=0.4}
    boneToModify {boneName=dyn_skirt_bk_side_l_a}
    constraintSetup [{transformType=Rotation, targetBone {boneName=thigh_l}}]
    constraintWeights [1]
   }
   n76 = Control Rig!(
    Source=Component To Local(ComponentPose=@n75)
    Alpha
    Alpha ads rope:double=Get Alpha_ads_rope() -> (Alpha_ads_rope:double)
    Alpha crouching:double=Get Alpha_crouching()
   ) -> (Pose) {
    alphaBoolBlend {blendOption=Linear}
    controlRigAssetReference {blueprintRigClass=&/M_MED_MetalScout_CtrlRig_rope.M_MED_MetalScout_CtrlRig_rope_C}
   }
   Output Pose(Result=Component To Local(ComponentPose=Local To Component(LocalPose=@n76))) {layerGroup=None}
  ]}
 EventGraph {nodes [
   Event BlueprintUpdateAnimation()
   > n1 = Sequence!(execute) -> (then_0, then_1, then_2:Exec, then_3:Exec, then_4:Exec, then_5:Exec, then_6:Exec, then_7:Exec, then_8:Exec, then_9:Exec)
   Set LOD_2_3(execute=@n1.then_9, LOD_2_3:bool="integer >= integer"(A=Get CurrentLOD(), B=2)) -> (Output_Get:bool)
   Set Alpha_frontend(execute=@n1.then_3, Alpha_frontend:double="To Float (Boolean)"(InBool=Get bIsInFrontEnd())) -> (Output_Get:double)
   Set Alpha_move_2d(execute=@n1.then_1, Alpha_move_2d:double="To Float (Boolean)"(InBool=Get bIsMoving2D())) -> (Output_Get:double)
   Set Alpha_skydiving(execute=@n1.then_2, Alpha_skydiving:double="To Float (Boolean)"(InBool=Get bIsSkyDiving())) -> (Output_Get:double)
   n2 = "To Float (Boolean)"(InBool=Get bIsCrouching())
   Set Alpha_crouching(execute=@n1.then_4, Alpha_crouching:double=@n2) -> (Output_Get:double)
   > Set inv_Alpha_crouching(inv_Alpha_crouching:double=MapRangeClamped(Value=@n2, InRangeA=0, InRangeB=1, OutRangeA=1, OutRangeB=0)) -> (Output_Get:double)
   n3 = GetOwnerSkeletalMeshComponent()
   Set Alpha_thigh_r_skirt_crouch(
    execute=@n1.then_5
    Alpha_thigh_r_skirt_crouch:double=float * float(
     A=@n2
     B=GetAlphaFromReferenceTransformDelta(SkeletalMeshComponent=@n3, BoneName=pelvis, BaseName=thigh_l, DeltaComponentType=QuaternionTwist, InputRange=(X=-20,Y=10), TwistAxis="0,1,0")
    ) {is=title K2Node_PromotableOperator}
   ) -> (Output_Get:double)
   Set Alpha_thigh_r_skirt(
    execute=@n1.then_7
    Alpha_thigh_r_skirt:double=GetAlphaFromReferenceTransformDelta(SkeletalMeshComponent=@n3, BoneName=thigh_r, BaseName=calf_r, DeltaComponentType=QuaternionTwist, InputRange=(X=30,Y=100), TwistAxis="0,0,1")
   ) -> (Output_Get:double)
   Set Alpha_elbowpad_l(
    execute=@n1.then_8
    Alpha_elbowpad_l:double=GetAlphaFromReferenceTransformDelta(
     SkeletalMeshComponent=@n3
     BoneName=upperarm_l
     BaseName=lowerarm_l
     DeltaComponentType=QuaternionTwist
     InputRange=(X=-30,Y=60)
     TwistAxis="0,0,1"
    )
   ) -> (Output_Get:double)
   n4 = Gravity Override State Function DEPRECATED(
    execute=@n1.then_0
    CharacterPartAnimInstance=Self-Reference()
    GameTimeInSeconds=GetGameTimeInSeconds()
    GravityOverrideParams=Get Gravity Override Params()
   )
   > Set Gravity(Gravity:Vector=@n4) -> (Output_Get:Vector)
   n5 = GetTwistAndSwingAngleOfDeltaRotationFromRefPose(self=@n3, BoneName=thigh_l)
   n6 = Get Alpha_crouching()
   Set Alpha_thigh_l_rope_crouch(execute=@n1.then_6, Alpha_thigh_l_rope_crouch:double=float * float(A=MapRangeClamped(Value=@n5, InRangeA=40, InRangeB=0), B=@n6)) -> (Output_Get:double)
   n7 = MapRangeClamped(Value=@n6, InRangeA=0, InRangeB=1, OutRangeA=1, OutRangeB=0)
   > Set Alpha_sprint_rope(Alpha_sprint_rope:double=MapRangeClamped(Value=@n5, InRangeA=10, InRangeB=30, OutRangeB=@n7)) -> (Output_Get:double)
   > Set Alpha_move_rope(Alpha_move_rope:double=float * float(A=@n7, B=Get Alpha_move_2d() -> (Alpha_move_2d:double))) -> (Output_Get:double)
   n8 = Get bIsTargeting()
   > Set Alpha_ads_rope(Alpha_ads_rope:double=float * float(A=Get inv_Alpha_crouching(), B="To Float (Boolean)"(InBool=@n8))) -> (Output_Get:double)
  ]}
]
variables [
 Alpha_ads_rope {type=number}
 Alpha_crouching {type=number}
 Alpha_elbowpad_l {type=number}
 Alpha_frontend {type=number}
 Alpha_move_2d {type=number}
 Alpha_move_rope {type=number}
 Alpha_skydiving {type=number}
 Alpha_sprint_rope {type=number}
 Alpha_thigh_l_rope_crouch {type=number}
 Alpha_thigh_r_skirt {type=number}
 Alpha_thigh_r_skirt_crouch {type=number}
 Gravity {type=Vector}
 Gravity Override Params {type=GravityOverrideParamsStruct}
 inv_Alpha_crouching {type=number}
 LOD_2_3 {type=boolean}
]
