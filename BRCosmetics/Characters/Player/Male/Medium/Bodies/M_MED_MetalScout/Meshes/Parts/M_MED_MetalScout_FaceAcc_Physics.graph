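# Physics asset for the MetalScout male medium body's face accessory mesh, giving its dynamic hair bones capsule bodies and joint limits for secondary motion.
# One head capsule is the hub, with most hair bones constrained directly to it.
# Strands A, B and K form two-link chains.
# All joints allow almost no twist, while swing varies per strand, with A and B nearly hinge-like through a very tight first swing limit.
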PhysicsAsset /BRCosmetics/Characters/Player/Male/Medium/Bodies/M_MED_MetalScout/Meshes/Parts/M_MED_MetalScout_FaceAcc_Physics
bodies {
 dyn_A_hair_01 [{
   shapeType=Capsule
   length=1.5696
   radius=0.7863
   shapeName=dyn_A_hair_01_capsule
   center=<1.06 0.135 0.0856>
   rotation=<rot 23.9336 97.8348 -88.7846>
  }]
 dyn_A_hair_02 [{
   shapeType=Capsule
   length=2.0562
   radius=0.6377
   shapeName=dyn_A_hair_02_capsule
   center=<1.5759 0.6139 0.7111>
   rotation=<rot 0 120.8267 -64.4834>
  }]
 dyn_B_hair_01 [{
   shapeType=Capsule
   length=2.0296
   radius=0.9415
   shapeName=dyn_B_hair_01_capsule
   center=<1.9809 0.0422 -0.0091>
   rotation=<rot 0.4316 91.6085 -88.4132>
  }]
 dyn_B_hair_02 [{
   shapeType=Capsule
   length=3.125
   radius=0.8745
   shapeName=dyn_B_hair_02_capsule
   center=<2.651 0.6196 0.216>
   rotation=<rot 0 107.5772 -81.9772>
  }]
 dyn_C_hair_01 [{
   shapeType=Capsule
   length=4.2066
   radius=2.4576
   shapeName=dyn_C_hair_02_capsule
   center=<1.4173 -0.102 0.1379>
   rotation=<rot 13.3221 -81.8045 -80.5896>
  }]
 dyn_D_hair_01 [{
   shapeType=Capsule
   length=3.1235
   radius=1.4756
   shapeName=dyn_C_hair_02_capsule
   center=<1.8396 -0.1518 0.0813>
   rotation=<rot 0 -96.7219 -82.4202>
  }]
 dyn_E_hair_01 [{
   shapeType=Capsule
   length=1.5845
   radius=0.9608
   shapeName=dyn_C_hair_02_capsule
   center=<1.3048 -0.0157 0.1541>
   rotation=<rot 0 -96.7219 -82.4202>
  }]
 dyn_H_hair_01 [{
   shapeType=Capsule
   length=2.5967
   radius=2.2451
   shapeName=dyn_C_hair_02_capsule
   center=<2.1304 -0.113 0.0435>
   rotation=<rot 0 -96.7219 -82.4202>
  }]
 dyn_I_hair_01 [{
   shapeType=Capsule
   length=2.9987
   radius=1.9515
   shapeName=dyn_C_hair_02_capsule
   center=<2.4586 -0.1517 -0.0005>
   rotation=<rot 0 -96.7219 -82.4202>
  }]
 dyn_J_hair_01 [{
   shapeType=Capsule
   length=1.6448
   radius=0.7216
   shapeName=dyn_D_hair_01_capsule
   center=<1.0029 0.0735 -0.0506>
   rotation=<rot 0 -88.2844 -86.0444>
  }]
 dyn_K_hair_01 [{
   shapeType=Capsule
   length=3.5665
   radius=2.3119
   shapeName=dyn_C_hair_02_capsule
   center=<2.8728 -0.2736 0.4427>
   rotation=<rot 0 -96.7219 -98.1041>
  }]
 dyn_K_hair_02_end [{
   shapeType=Capsule
   length=1.9373
   radius=1.808
   shapeName=dyn_C_hair_02_capsule
   center=<2.3229 -0.2088 0.0166>
   rotation=<rot 0 -96.7219 -82.4202>
  }]
 dyn_L_hair_01 [{
   shapeType=Capsule
   length=3.1991
   radius=1.9897
   shapeName=dyn_C_hair_02_capsule
   center=<3.0229 -0.2913 -0.2884>
   rotation=<rot 0 -96.7219 -82.4202>
  }]
 dyn_M_hair_01 [{
   shapeType=Capsule
   length=2.3356
   radius=1.8799
   shapeName=dyn_C_hair_02_capsule
   center=<1.6382 -0.055 -0.0262>
   rotation=<rot 0 -96.7219 -88.5552>
  }]
 dyn_N_hair_01 [{
   shapeType=Capsule
   length=2.8962
   radius=1.9531
   shapeName=dyn_C_hair_02_capsule
   center=<2.944 -0.2089 -0.0593>
   rotation=<rot 0 -96.7219 -88.5552>
  }]
 dyn_O_hair_01 [{
   shapeType=Capsule
   length=2.9564
   radius=1.7656
   shapeName=dyn_C_hair_02_capsule
   center=<2.8202 -0.2674 -0.0501>
   rotation=<rot 0 -96.7219 -82.4202>
  }]
 dyn_P_hair_01 [{
   shapeType=Capsule
   length=3.5665
   radius=2.3119
   shapeName=dyn_C_hair_02_capsule
   center=<2.8728 -0.2736 0.4427>
   rotation=<rot 0 -96.7219 -98.1041>
  }]
 dyn_Q_hair_01 [{
   shapeType=Capsule
   length=3.0508
   radius=1.8068
   shapeName=dyn_C_hair_02_capsule
   center=<2.4855 -0.2279 -0.0052>
   rotation=<rot 0 -96.7219 -82.4202>
  }]
 dyn_R_hair_01 [{
   shapeType=Capsule
   length=2.682
   radius=1.8223
   shapeName=dyn_C_hair_02_capsule
   center=<2.2631 -0.2017 0.0246>
   rotation=<rot 0 -96.7219 -82.4202>
  }]
 dyn_S_hair_01 [{
   shapeType=Capsule
   length=3.0804
   radius=2.003
   shapeName=dyn_C_hair_02_capsule
   center=<2.7882 -0.2636 -0.0458>
   rotation=<rot 0 -96.7219 -82.4202>
  }]
 dyn_T_hair_01 [{
   shapeType=Capsule
   length=1.9373
   radius=1.808
   shapeName=dyn_C_hair_02_capsule
   center=<2.3229 -0.2088 0.0166>
   rotation=<rot 0 -96.7219 -82.4202>
  }]
 dyn_U_hair_01 [{
   shapeType=Capsule
   length=2.7094
   radius=1.7489
   shapeName=dyn_C_hair_02_capsule
   center=<2.0894 -0.1813 0.0479>
   rotation=<rot 0 -96.7219 -82.4202>
  }]
 dyn_V_hair_01 [{
   shapeType=Capsule
   length=3.0804
   radius=2.003
   shapeName=dyn_C_hair_02_capsule
   center=<2.7882 -0.2636 -0.0458>
   rotation=<rot 0 -96.7219 -82.4202>
  }]
 dyn_W_hair_01 [{
   shapeType=Capsule
   length=2.682
   radius=1.8223
   shapeName=dyn_C_hair_02_capsule
   center=<2.2631 -0.2017 0.0246>
   rotation=<rot 0 -96.7219 -82.4202>
  }]
 dyn_X_hair_01 [{
   shapeType=Capsule
   length=3.0508
   radius=1.8068
   shapeName=dyn_C_hair_02_capsule
   center=<2.4855 -0.2279 -0.0052>
   rotation=<rot 0 -96.7219 -82.4202>
  }]
 head [{shapeType=Capsule, length=5.7553, radius=6.3815, shapeName=head_capsule, rotation=<rot 9.5694 93.7078 -88.105>}]
}
constraints [
 {bone1Name=dyn_B_hair_02, bone2Name=dyn_B_hair_01, swing1LimitDegrees=0.5, swing2LimitDegrees=45, twistLimitDegrees=2}
 {bone1Name=dyn_A_hair_02, bone2Name=dyn_A_hair_01, swing1LimitDegrees=0.5, swing2LimitDegrees=60, twistLimitDegrees=2}
 {bone1Name=dyn_B_hair_01, bone2Name=head, swing1LimitDegrees=0.5, swing2LimitDegrees=45, twistLimitDegrees=2}
 {bone1Name=dyn_A_hair_01, bone2Name=head, swing1LimitDegrees=0.5, swing2LimitDegrees=60, twistLimitDegrees=2}
 {bone1Name=dyn_C_hair_01, bone2Name=head, swing1LimitDegrees=60, swing2LimitDegrees=60, twistLimitDegrees=2}
 {bone1Name=dyn_D_hair_01, bone2Name=head, swing1LimitDegrees=65, swing2LimitDegrees=30, twistLimitDegrees=2}
 {bone1Name=dyn_E_hair_01, bone2Name=head, swing1LimitDegrees=45, swing2LimitDegrees=10, twistLimitDegrees=2}
 {bone1Name=dyn_H_hair_01, bone2Name=head, swing1LimitDegrees=30, swing2LimitDegrees=60, twistLimitDegrees=2}
 {bone1Name=dyn_I_hair_01, bone2Name=head, swing1LimitDegrees=30, swing2LimitDegrees=60, twistLimitDegrees=2}
 {bone1Name=dyn_J_hair_01, bone2Name=head, swing1LimitDegrees=5, swing2LimitDegrees=45, twistLimitDegrees=2}
 {bone1Name=dyn_M_hair_01, bone2Name=head, swing1LimitDegrees=30, swing2LimitDegrees=60, twistLimitDegrees=2}
 {bone1Name=dyn_L_hair_01, bone2Name=head, swing1LimitDegrees=10, swing2LimitDegrees=30, twistLimitDegrees=2}
 {bone1Name=dyn_N_hair_01, bone2Name=head, swing1LimitDegrees=50, swing2LimitDegrees=30, twistLimitDegrees=2}
 {bone1Name=dyn_O_hair_01, bone2Name=head, swing1LimitDegrees=60, swing2LimitDegrees=30, twistLimitDegrees=2}
 {bone1Name=dyn_P_hair_01, bone2Name=head, swing1LimitDegrees=30, swing2LimitDegrees=30, twistLimitDegrees=2}
 {bone1Name=dyn_Q_hair_01, bone2Name=head, swing1LimitDegrees=80, swing2LimitDegrees=30, twistLimitDegrees=2}
 {bone1Name=dyn_R_hair_01, bone2Name=head, swing1LimitDegrees=30, swing2LimitDegrees=60, twistLimitDegrees=2}
 {bone1Name=dyn_S_hair_01, bone2Name=head, swing1LimitDegrees=65, swing2LimitDegrees=30, twistLimitDegrees=2}
 {bone1Name=dyn_T_hair_01, bone2Name=head, swing1LimitDegrees=30, swing2LimitDegrees=30, twistLimitDegrees=2}
 {bone1Name=dyn_U_hair_01, bone2Name=head, swing1LimitDegrees=45, swing2LimitDegrees=10, twistLimitDegrees=2}
 {bone1Name=dyn_V_hair_01, bone2Name=head, swing1LimitDegrees=65, swing2LimitDegrees=30, twistLimitDegrees=2}
 {bone1Name=dyn_W_hair_01, bone2Name=head, swing1LimitDegrees=60, swing2LimitDegrees=30, twistLimitDegrees=2}
 {bone1Name=dyn_X_hair_01, bone2Name=head, swing1LimitDegrees=80, swing2LimitDegrees=30, twistLimitDegrees=2}
 {bone1Name=dyn_K_hair_01, bone2Name=head, swing1LimitDegrees=70, swing2LimitDegrees=30, twistLimitDegrees=2}
 {bone1Name=dyn_K_hair_02_end, bone2Name=dyn_K_hair_01, swing1LimitDegrees=70, swing2LimitDegrees=30, twistLimitDegrees=2}
]
