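# Physics asset for the face accessory part of the male medium MetalScout player body, covering dangling hair, earrings and a tail.
# Two two-segment hair chains, left and right earrings and a three-segment tail hang from the head, which links to spine_05.
# Capsules shape every body, with an added box on the upper spine.
# Hair swings mostly on one axis, earrings swing in cones, and the tail root and head-to-spine joints rotate freely.
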PhysicsAsset /BRCosmetics/Characters/Player/Male/Medium/Bodies/M_MED_MetalScout/Meshes/Parts/M_MED_MetalScout_FaceAcc_Physics_earrings1
bodies {
 dyn_A_hair_01 [{
   shapeType=Capsule
   length=1.5696
   radius=0.7863
   shapeName=dyn_A_hair_01_capsule
   center=<1.06 0.135 0.0856>
   rotation=<rot 23.9336 97.8348 -88.7846>
  }]
 dyn_A_hair_02 [{
   shapeType=Capsule
   length=2.0562
   radius=0.6377
   shapeName=dyn_A_hair_02_capsule
   center=<1.5759 0.6139 0.7111>
   rotation=<rot 0 120.8267 -64.4834>
  }]
 dyn_B_hair_01 [{
   shapeType=Capsule
   length=2.0296
   radius=0.9415
   shapeName=dyn_B_hair_01_capsule
   center=<1.9809 0.0422 -0.0091>
   rotation=<rot 0.4316 91.6085 -88.4132>
  }]
 dyn_B_hair_02 [{
   shapeType=Capsule
   length=3.125
   radius=0.8745
   shapeName=dyn_B_hair_02_capsule
   center=<2.651 0.6196 0.216>
   rotation=<rot 0 107.5772 -81.9772>
  }]
 dyn_L_earring_01 [{
   shapeType=Capsule
   length=2.7719
   radius=0.5
   shapeName=dyn_L_earring_01_capsule
   center=<2.183 -0.0009 -0.0399>
   rotation=<rot 0 -90.2257 -89.0787>
  }]
 dyn_R_earring_01 [{
   shapeType=Capsule
   length=2.7709
   radius=0.5
   shapeName=dyn_R_earring_01_capsule
   center=<-2.1826 0.0009 0.0413>
   rotation=<rot 0 -90.1975 -88.894>
  }]
 dyn_tail_01_joint [{
   shapeType=Capsule
   length=1.7329
   radius=1.6994
   shapeName=dyn_tail_01_joint_capsule
   center=<0.4225 -0.001 -0.0035>
   rotation=<rot -1.9174 -90 0.6519>
  }]
 dyn_tail_02_joint [{
   shapeType=Capsule
   length=3.3278
   radius=2.5717
   shapeName=dyn_B_hair_01_capsule
   center=<3.7015 0.0902 0.0385>
   rotation=<rot 0.4316 91.6085 -88.4132>
  }]
 dyn_tail_03_joint [{
   shapeType=Capsule
   length=3.9594
   radius=2.8844
   shapeName=dyn_B_hair_02_capsule
   center=<4.1144 1.0832 0.4324>
   rotation=<rot 0 107.5772 -81.9772>
  }]
 head [{shapeType=Capsule, length=5.7553, radius=6.3815, shapeName=head_capsule, rotation=<rot 9.5694 93.7078 -88.105>}]
 spine_05 [
  {shapeType=Capsule, length=8.556, radius=11.1386, shapeName=spine_05_capsule, center=<6.5481 16.6126 0>, rotation=<rot 0 -79.811 90>}
  {shapeType=Box, extentX=13.8284, extentY=7.5, extentZ=19.4297, shapeName=spine_05_box, center=<9.5437 5.9186 0>, rotation=<rot 0 30.0701 0>}
 ]
}
constraints [
 {bone1Name=dyn_A_hair_01, bone2Name=head, swing1LimitDegrees=0.5, swing2LimitDegrees=60, twistLimitDegrees=2}
 {bone1Name=dyn_A_hair_02, bone2Name=dyn_A_hair_01, swing1LimitDegrees=0.5, swing2LimitDegrees=60, twistLimitDegrees=2}
 {bone1Name=dyn_B_hair_01, bone2Name=head, swing1LimitDegrees=0.5, swing2LimitDegrees=45, twistLimitDegrees=2}
 {bone1Name=dyn_B_hair_02, bone2Name=dyn_B_hair_01, swing1LimitDegrees=0.5, swing2LimitDegrees=45, twistLimitDegrees=2}
 {bone1Name=dyn_R_earring_01, bone2Name=head, swing1LimitDegrees=45, swing2LimitDegrees=45, twistLimitDegrees=10}
 {bone1Name=dyn_L_earring_01, bone2Name=head, swing1LimitDegrees=45, swing2LimitDegrees=45, twistLimitDegrees=10}
 {
  bone1Name=dyn_tail_01_joint
  bone2Name=head
  swing1LimitDegrees=45
  swing1Motion=Free
  swing2LimitDegrees=45
  swing2Motion=Free
  twistLimitDegrees=45
  twistMotion=Free
 }
 {bone1Name=dyn_tail_02_joint, bone2Name=dyn_tail_01_joint, swing1LimitDegrees=30, swing2LimitDegrees=5, twistLimitDegrees=2}
 {bone1Name=dyn_tail_03_joint, bone2Name=dyn_tail_02_joint, swing1LimitDegrees=10, swing2LimitDegrees=5, twistLimitDegrees=2}
 {
  bone1Name=head
  bone2Name=spine_05
  swing1LimitDegrees=45
  swing1Motion=Free
  swing2LimitDegrees=45
  swing2Motion=Free
  twistLimitDegrees=45
  twistMotion=Free
 }
]
