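# Post-pose AnimBlueprint for the male MetalScout face accessory, adding secondary motion to earrings and a dynamic tail.
# It copies the parent mesh pose through two neck-relative RigidBody simulations, then additively bends and shifts tail joints.
# Head twist against spine_05 drives the tail correctives, with a stronger root shift while targeting.
# Skydiving flips simulation gravity upward, and jumping or falling during tactical sprint lowers physics alpha.
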AnimBlueprint /BRCosmetics/Characters/Player/Male/Medium/Bodies/M_MED_MetalScout/Meshes/Parts/M_MED_MetalScout_FaceAcc_AnimBP
target_skeleton=&/M_MED_MetalScout_FaceAcc_Skeleton
graphs [
 AnimGraph {nodes [
   n1 = Get override gravity() -> (override gravity:Vector)
   n2 = Get alpha() -> (alpha:double)
   n3 = RigidBody!(
    ComponentPose=Local To Component(LocalPose=Copy Pose From Mesh() {bUseAttachedParent=true})
    Alpha=@n2
    OverrideWorldGravity:Vector=@n1
    ExternalForce
   ) -> (Pose) {
    bOverrideWorldGravity=true
    overridePhysicsAsset=&/M_MED_MetalScout_FaceAcc_Physics
    simulationSpace=BaseBoneSpace
    alphaBoolBlend {blendOption=Linear}
    baseBoneRef {boneName=neck_01}
    componentLinearAccScale=<1 1 1>
    componentLinearVelScale=<1 1 1>
    overrideWorldGravity {z=-980}
    simSpaceSettings {maxAngularAcceleration=500, maxAngularVelocity=3, maxLinearAcceleration=400, maxLinearVelocity=250}
   }
   n4 = RigidBody(ComponentPose=@n3, Alpha=@n2, OverrideWorldGravity=@n1) {
    bOverrideWorldGravity=true
    overridePhysicsAsset=&/M_MED_MetalScout_FaceAcc_Physics_earrings1
    simulationSpace=BaseBoneSpace
    alphaBoolBlend {blendOption=Linear}
    baseBoneRef {boneName=neck_02}
    componentAppliedLinearAccClamp=<500 500 500>
    componentLinearAccScale=<0.8 0.8 0.8>
    componentLinearVelScale=<0.8 0.8 0.8>
    overrideWorldGravity {z=-980}
    simSpaceSettings {maxAngularAcceleration=500, maxAngularVelocity=3, maxLinearAcceleration=400, maxLinearVelocity=250}
   }
   n5 = Get tail() -> (tail:double)
   n6 = "Transform (Modify) Bone - Bone: dyn_tail_01_joint"(ComponentPose=@n4, Alpha=@n5, Translation="0,0,0", Rotation="0,0,10", Scale="1,1,1") {rotationMode=BMM_Additive, alphaBoolBlend {blendOption=Linear}, boneToModify {boneName=dyn_tail_01_joint}, rotation {roll=10}}
   n7 = "Transform (Modify) Bone - Bone: dyn_tail_02_joint"(ComponentPose=@n6, Alpha=@n5, Translation="0,0,0", Rotation="0,0,60", Scale="1,1,1") {rotationMode=BMM_Additive, alphaBoolBlend {blendOption=Linear}, boneToModify {boneName=dyn_tail_02_joint}, rotation {roll=60}}
   n8 = "Transform (Modify) Bone - Bone: dyn_tail_03_joint"(ComponentPose=@n7, Alpha=@n5, Translation="0,0,0", Rotation="0,0,25", Scale="1,1,1") {rotationMode=BMM_Additive, alphaBoolBlend {blendOption=Linear}, boneToModify {boneName=dyn_tail_03_joint}, rotation {roll=25}}
   n9 = "Transform (Modify) Bone - Bone: dyn_tail_01_joint"(ComponentPose=@n8, Alpha=Get tail root() -> (tail root:double), Translation="0,-1,0.3", Rotation="0,0,0", Scale="1,1,1") {
    rotationMode=BMM_Additive
    translationMode=BMM_Additive
    alphaBoolBlend {blendOption=Linear}
    boneToModify {boneName=dyn_tail_01_joint}
    translation {y=-1, z=0.3}
   }
   Output Pose(Result=Component To Local(ComponentPose=@n9)) {layerGroup=None}
  ]}
 EventGraph {nodes [
   Event BlueprintUpdateAnimation()
   > n1 = Sequence!(execute) -> (then_0, then_1, then_2:Exec, then_3:Exec, then_4:Exec, then_5:Exec, then_6:Exec, then_7:Exec)
   > n2 = Branch(Condition=Get bIsSkyDiving())
   > Set override gravity(override gravity:Vector="0,0,100") -> (Output_Get:Vector)
   Set override gravity(execute=@n2.else, override gravity="0,0,-800")
   n3 = GetAlphaFromReferenceTransformDelta(SkeletalMeshComponent=GetOwnerSkeletalMeshComponent(), BoneName=head, BaseName=spine_05, InputRange=(X=0,Y=-53), TwistAxis="1,0,0")
   n4 = float * float(A=@n3, B=0.5) {is=title K2Node_PromotableOperator}
   Set tail(execute=@n1.then_1, tail:double=@n4) -> (Output_Get:double)
   > n5 = Branch(Condition=Get bIsTargeting())
   > Set tail root(tail root:double=@n3) -> (Output_Get:double)
   Set tail root(execute=@n5.else, tail root=@n4)
   n6 = Get bIsTacticalSprinting()
   n7 = Get bIsFalling()
   n8 = Get bIsJumping()
   n9 = Branch(execute=@n1.then_2, Condition=OR Boolean(A=AND Boolean(A=@n6, B=@n8), B=AND Boolean(A=@n6, B=@n7)))
   > Set alpha(alpha:double=0.3) -> (Output_Get:double)
   Set alpha(execute=@n9.else, alpha=1)
  ]}
]
variables [
 alpha {type=number}
 override gravity {type=Vector}
 tail {type=number}
 tail root {type=number}
]
